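# Control Rig for the female medium FelineWarrior body that applies corrective deformation and poses clothing bones.
# A sequence runs the shared DeformRig function, collar and skirt correctives, then arm-string and belt parent constraints.
# Spherical pose readers on head, upper arms and thighs weight relative-transform offsets, some through remaps.
# Belt loop and rope correctives run only when `cr` is false; sleeve and scarf chains are unconnected to execution.
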ControlRigBlueprint /Game/Characters/Player/Female/Medium/Bodies/F_MED_FelineWarrior/Meshes/F_MED_FelineWarrior_CtrlRig
graphs [
 RigVMFunctionLibrary {links [], nodes []}
 RigVMFunction_Sequence_1_ContainedGraph {nodes [
   n1 = RigVMFunctionEntryNode()
   RigVMFunction_Sequence = &RigVMFunction_Sequence(ExecuteContext=@n1.ExecuteContext)
   n2 = &RigVMFunction_Sequence(ExecuteContext=@RigVMFunction_Sequence.B)
   n3 = &RigVMFunction_Sequence(ExecuteContext=@n2.B)
   n4 = &RigVMFunction_Sequence(ExecuteContext=@n3.B)
   RigVMFunctionReturnNode(A=@RigVMFunction_Sequence.A, B=@n2.A, C=@n3.A, D=@n4.A, E=@n4.B)
  ]}
 RigVMModel {nodes [
   RigUnit_BeginExecution = &RigUnit_BeginExecution()
   RigVMFunction_Sequence = RigVMAggregateNode(ExecuteContext=@RigUnit_BeginExecution.ExecutePin)
   FN_DeformRig_FNC = RigVMFunctionReferenceNode(
    DeformRig_ Curve="(deformRig_l_upperarm_up_alpha_72_#1=False,deformRig_r_upperarm_up_alpha_73_#2=False,deformRig_l_upperarm_fr_alpha_74_#3=False,deformRig_r_upperarm_fr_alpha_75_CB51B8 ...(truncated)"
    DeformRig_Bool="(upperarm_1_#4=True,elbow_3_#5=True,wrist_7_#6=True,groin_8_#7=true,glute_17_#8=false,knee_9_40C408324CA26 ...(truncated)"
    DeformRig_Debug=(upperarm_debug_18_#4=False,wrist_debug_20_#6=False,lat_debug_26_#9=False,groinGlute_debug_27_#7=False)
    DeformRig_Value="(upperarm_up_18_#10=(Rotation=(X=0,Y=0,Z=0,W=1),Translation=(X=0,Y=0,Z=10),Scale3D=(X=0.7,Y=1,Z=1)),upperarm_fwd_fr_19_#11=(Rotation=(X=0,Y=0,Z=0,W=1),Translation=(X=-10,Y=5,Z=5),Scale3D=(X=1,Y=1,Z= ...(truncated)"
    ExecuteContext=@RigVMFunction_Sequence.A
   )
   n1 = &RigUnit_SphericalPoseReader(
    ActiveRegionScaleFactors=(PositiveWidth=1,NegativeWidth=1.5,PositiveHeight=2.1,NegativeHeight=2.1)
    ActiveRegionSize=0.15
    Debug=(bDrawDebug=false,bDraw2D=false,bDrawLocalAxes=false,DebugScale=25,DebugSegments=20,DebugThickness=0.25)
    DriverAxis=(X=1,Y=0,Z=0)
    DriverItem=(Type=Bone,Name="head")
    FalloffRegionScaleFactors=(PositiveWidth=4,NegativeWidth=1.5,PositiveHeight=-5,NegativeHeight=-5)
    FalloffSize=0.3
    FlipHeightScaling=false
    FlipWidthScaling=false
    OptionalParentItem=(Type=Bone,Name="spine_05")
    RotationOffset=(X=-49.1172,Y=1.4611,Z=-28.8422)
    ExecutePin=@RigVMFunction_Sequence.B
   )
   > RigUnit_SetRelativeTransformForItem = /Script/ControlRig.RigUnit_SetRelativeTransformForItem(
    Child=(Type=Bone,Name="base_back_collar")
    Parent=(Type=Bone,Name="base_back_collar")
    Value=(Rotation=(X=-0.2442,Y=-0.1266,Z=-0.6457,W=0.7123),Translation=(X=-0.1172,Y=-0.0479,Z=-0.0134),Scale3D=(X=1,Y=1,Z=1))
    bParentInitial=False
    bPropagateToChildren=True
    Weight=@n1.OutputParam
   ) {
    is=struct RigVMUnitNode
    template="Set Relative Transform::Execute(in Child,in Parent,in bParentInitial,in Value,in Weight,in bPropagateToChildren)"
   }
   > &RigUnit_SetRelativeTransformForItem(
    Child=(Type=Bone,Name="base_front_collar_r")
    Parent=(Type=Bone,Name="base_front_collar_r")
    Value=(Rotation=(X=0,Y=0,Z=0,W=1),Translation=(X=0,Y=0,Z=0),Scale3D=(X=1,Y=1,Z=1))
    bParentInitial=False
    bPropagateToChildren=True
    Weight=@n1.OutputParam
   ) {template="Set Relative Transform::Execute(in Child,in Parent,in bParentInitial,in Value,in Weight,in bPropagateToChildren)"}
   > &RigUnit_SetRelativeTransformForItem(
    Child=(Type=Bone,Name="side_collar_r")
    Parent=(Type=Bone,Name="side_collar_r")
    Value=(Rotation=(X=-0.3296,Y=0.3197,Z=0.0529,W=0.8868),Translation=(X=0,Y=0,Z=0),Scale3D=(X=1,Y=1,Z=1))
    bParentInitial=False
    bPropagateToChildren=True
    Weight=@n1.OutputParam
   ) {template="Set Relative Transform::Execute(in Child,in Parent,in bParentInitial,in Value,in Weight,in bPropagateToChildren)"}
   > &RigUnit_SetRelativeTransformForItem(
    Child=(Type=Bone,Name="side_collar_l")
    Parent=(Type=Bone,Name="side_collar_l")
    Value=(Rotation=(X=0.383,Y=0.3957,Z=-0.4904,W=0.6755),Translation=(X=0,Y=0,Z=0),Scale3D=(X=1,Y=1,Z=1))
    bParentInitial=False
    bPropagateToChildren=True
    Weight=@n1.OutputParam
   ) {template="Set Relative Transform::Execute(in Child,in Parent,in bParentInitial,in Value,in Weight,in bPropagateToChildren)"}
   > n2 = &RigUnit_SphericalPoseReader(
    ActiveRegionScaleFactors=(PositiveWidth=1.8,NegativeWidth=1,PositiveHeight=1.5,NegativeHeight=1.5)
    ActiveRegionSize=0.15
    Debug=(bDrawDebug=false,bDraw2D=false,bDrawLocalAxes=false,DebugScale=25,DebugSegments=20,DebugThickness=0.25)
    DriverAxis=(X=1,Y=0,Z=-0.187)
    DriverItem=(Type=Bone,Name="head")
    FalloffRegionScaleFactors=(PositiveWidth=5,NegativeWidth=1.5,PositiveHeight=5,NegativeHeight=2)
    FalloffSize=0.3
    FlipHeightScaling=false
    FlipWidthScaling=false
    OptionalParentItem=(Type=Bone,Name="spine_05")
    RotationOffset=(X=-157.4494,Y=-9.6082,Z=61.8156)
   )
   > &RigUnit_SetRelativeTransformForItem(
    Child=(Type=Bone,Name="base_back_collar")
    Parent=(Type=Bone,Name="base_back_collar")
    Value=(Rotation=(X=-0.1039,Y=0.0526,Z=-0.2394,W=0.9639),Translation=(X=0,Y=0,Z=0),Scale3D=(X=1,Y=1,Z=1))
    bParentInitial=False
    bPropagateToChildren=True
    Weight=@n2.OutputParam
   ) {template="Set Relative Transform::Execute(in Child,in Parent,in bParentInitial,in Value,in Weight,in bPropagateToChildren)"}
   > &RigUnit_SetRelativeTransformForItem(
    Child=(Type=Bone,Name="side_collar_l")
    Parent=(Type=Bone,Name="side_collar_l")
    Value=(Rotation=(X=-0.0023,Y=0.2343,Z=-0.0788,W=0.969),Translation=(X=0,Y=0,Z=0),Scale3D=(X=1,Y=1,Z=1))
    bParentInitial=False
    bPropagateToChildren=True
    Weight=@n2.OutputParam
   ) {template="Set Relative Transform::Execute(in Child,in Parent,in bParentInitial,in Value,in Weight,in bPropagateToChildren)"}
   > n3 = &RigUnit_SphericalPoseReader(
    ActiveRegionScaleFactors=(PositiveWidth=1.2,NegativeWidth=1.2,PositiveHeight=1.2,NegativeHeight=1.2)
    ActiveRegionSize=0.15
    Debug=(bDrawDebug=false,bDraw2D=false,bDrawLocalAxes=false,DebugScale=25,DebugSegments=20,DebugThickness=0.25)
    DriverAxis=(X=0,Y=0,Z=1)
    DriverItem=(Type=Bone,Name="head")
    FalloffRegionScaleFactors=(PositiveWidth=1,NegativeWidth=1,PositiveHeight=1,NegativeHeight=1)
    FalloffSize=0.3
    FlipHeightScaling=false
    FlipWidthScaling=false
    OptionalParentItem=(Type=Bone,Name="spine_05")
    RotationOffset=(X=31.3444,Y=-45.4811,Z=52.8708)
   )
   > &RigUnit_SetRelativeTransformForItem(
    Child=(Type=Bone,Name="base_back_collar")
    Parent=(Type=Bone,Name="base_back_collar")
    Value=(Rotation=(X=0.1308,Y=-0.0196,Z=-0.1651,W=0.9774),Translation=(X=0,Y=0,Z=0),Scale3D=(X=1,Y=1,Z=1))
    bParentInitial=False
    bPropagateToChildren=True
    Weight=@n3.OutputParam
   ) {template="Set Relative Transform::Execute(in Child,in Parent,in bParentInitial,in Value,in Weight,in bPropagateToChildren)"}
   > n4 = &RigUnit_SphericalPoseReader(
    ActiveRegionScaleFactors=(PositiveWidth=3,NegativeWidth=3,PositiveHeight=1.5,NegativeHeight=0.75)
    ActiveRegionSize=0.1
    Debug=(bDrawDebug=false,bDrawLocalAxes=false,DebugScale=25,DebugSegments=20,DebugThickness=0.25)
    DriverAxis=(X=1,Y=0,Z=0)
    DriverItem=(Type=Bone,Name="upperarm_l")
    FalloffRegionScaleFactors=(PositiveWidth=1,NegativeWidth=1,PositiveHeight=1,NegativeHeight=1)
    FalloffSize=0.4
    OptionalParentItem=(Type=Bone)
    RotationOffset=(X=20,Y=-14.5226,Z=0)
   )
   > &RigUnit_SetRelativeTransformForItem(
    Child=(Type=Bone,Name="base_front_collar_l")
    Parent=(Type=Bone,Name="base_front_collar_l")
    Value=(Rotation=(X=-0.0025,Y=-0.1467,Z=-0.0245,W=0.9889),Translation=(X=0,Y=0,Z=0),Scale3D=(X=1,Y=1,Z=1))
    bParentInitial=False
    bPropagateToChildren=True
    Weight=@n4.OutputParam
   ) {template="Set Relative Transform::Execute(in Child,in Parent,in bParentInitial,in Value,in Weight,in bPropagateToChildren)"}
   > &RigUnit_SetRelativeTransformForItem(
    Child=(Type=Bone,Name="side_collar_l")
    Parent=(Type=Bone,Name="side_collar_l")
    Value=(Rotation=(X=-0.2544,Y=-0.0364,Z=0.014,W=0.9663),Translation=(X=0,Y=0,Z=0),Scale3D=(X=1,Y=1,Z=1))
    bParentInitial=False
    bPropagateToChildren=True
    Weight=@n4.OutputParam
   ) {template="Set Relative Transform::Execute(in Child,in Parent,in bParentInitial,in Value,in Weight,in bPropagateToChildren)"}
   > n5 = &RigUnit_SphericalPoseReader(
    ActiveRegionScaleFactors=(PositiveWidth=3,NegativeWidth=3,PositiveHeight=1.5,NegativeHeight=0.75)
    ActiveRegionSize=0.1
    Debug=(bDrawDebug=false,bDrawLocalAxes=false,DebugScale=25,DebugSegments=20,DebugThickness=0.25)
    DriverAxis=(X=-4.7256,Y=-2.429,Z=0.4527)
    DriverItem=(Type=Bone,Name="upperarm_r")
    FalloffRegionScaleFactors=(PositiveWidth=1,NegativeWidth=1,PositiveHeight=1,NegativeHeight=1)
    FalloffSize=0.4
    FlipWidthScaling=true
    OptionalParentItem=(Type=Bone)
    RotationOffset=(X=-156.5466,Y=21.5663,Z=25.1647)
   )
   > &RigUnit_SetRelativeTransformForItem(
    Child=(Type=Bone,Name="base_front_collar_r")
    Parent=(Type=Bone,Name="base_front_collar_r")
    Value=(Rotation=(X=0,Y=-0.5957,Z=0,W=0.8032),Translation=(X=0,Y=0,Z=0),Scale3D=(X=1,Y=1,Z=1))
    bParentInitial=False
    bPropagateToChildren=True
    Weight=@n5.OutputParam
   ) {template="Set Relative Transform::Execute(in Child,in Parent,in bParentInitial,in Value,in Weight,in bPropagateToChildren)"}
   n6 = &RigUnit_SphericalPoseReader(
    ActiveRegionScaleFactors=(PositiveWidth=3.5,NegativeWidth=0.5,PositiveHeight=3,NegativeHeight=3)
    ActiveRegionSize=0.1
    Debug=(bDrawDebug=false,DebugScale=25,DebugSegments=20,DebugThickness=0.25)
    DriverAxis=(X=-1,Y=0,Z=0)
    DriverItem=(Type=Bone,Name="thigh_l")
    FalloffRegionScaleFactors=(PositiveWidth=1,NegativeWidth=1.3,PositiveHeight=1,NegativeHeight=1)
    FalloffSize=0.45
    OptionalParentItem=(Type=Bone)
    RotationOffset=(X=120.1301,Y=-11.8479,Z=-12.0299)
    ExecutePin=@RigVMFunction_Sequence.C
   )
   > &RigUnit_SetRelativeTransformForItem(
    Child=(Type=Bone,Name="dyn_skirt_side_frnt_l_dynbase")
    Parent=(Type=Bone,Name="dyn_skirt_side_frnt_l_dynbase")
    Value=(Rotation=(X=-0.1739,Y=-0.0205,Z=0.0809,W=0.9812),Translation=(X=0,Y=0,Z=0),Scale3D=(X=1,Y=1,Z=1))
    bParentInitial=False
    bPropagateToChildren=True
    Weight=@n6.OutputParam
   ) {template="Set Relative Transform::Execute(in Child,in Parent,in bParentInitial,in Value,in Weight,in bPropagateToChildren)"}
   > &RigUnit_SetRelativeTransformForItem(
    Child=(Type=Bone,Name="dyn_skirt_side_frnt_r_dynbase")
    Parent=(Type=Bone,Name="dyn_skirt_side_frnt_r_dynbase")
    Value=(Rotation=(X=-0.1516,Y=0.1556,Z=0.1447,W=0.9653),Translation=(X=0,Y=0,Z=0),Scale3D=(X=1,Y=1,Z=1))
    bParentInitial=False
    bPropagateToChildren=True
    Weight=@n6.OutputParam
   ) {template="Set Relative Transform::Execute(in Child,in Parent,in bParentInitial,in Value,in Weight,in bPropagateToChildren)"}
   > &RigUnit_SetRelativeTransformForItem(
    Child=(Type=Bone,Name="dyn_skirt_side_l_base")
    Parent=(Type=Bone,Name="dyn_skirt_side_l_base")
    Value=(Rotation=(X=0.0445,Y=0.0397,Z=-0.1586,W=0.9855),Translation=(X=0,Y=0,Z=0),Scale3D=(X=1,Y=1,Z=1))
    bParentInitial=False
    bPropagateToChildren=True
    Weight=@n6.OutputParam
   ) {template="Set Relative Transform::Execute(in Child,in Parent,in bParentInitial,in Value,in Weight,in bPropagateToChildren)"}
   > n7 = &RigUnit_SphericalPoseReader(
    ActiveRegionScaleFactors=(PositiveWidth=3.5,NegativeWidth=0.5,PositiveHeight=3,NegativeHeight=3)
    ActiveRegionSize=0.1
    Debug=(bDrawDebug=false,DebugScale=25,DebugSegments=20,DebugThickness=0.25)
    DriverAxis=(X=1,Y=0,Z=0)
    DriverItem=(Type=Bone,Name="thigh_r")
    FalloffRegionScaleFactors=(PositiveWidth=1,NegativeWidth=1.3,PositiveHeight=2,NegativeHeight=1)
    FalloffSize=0.45
    FlipHeightScaling=false
    FlipWidthScaling=true
    OptionalParentItem=(Type=Bone)
    RotationOffset=(X=-56.34,Y=4.3495,Z=-12.2736)
   )
   > &RigUnit_SetRelativeTransformForItem(
    Child=(Type=Bone,Name="dyn_skirt_side_frnt_r_dynbase")
    Parent=(Type=Bone,Name="dyn_skirt_side_frnt_r_dynbase")
    Value=(Rotation=(X=0.1358,Y=0.0151,Z=-0.1893,W=0.9724),Translation=(X=0,Y=0,Z=0),Scale3D=(X=1,Y=1,Z=1))
    bParentInitial=False
    bPropagateToChildren=True
    Weight=@n7.OutputParam
   ) {template="Set Relative Transform::Execute(in Child,in Parent,in bParentInitial,in Value,in Weight,in bPropagateToChildren)"}
   > &RigUnit_SetRelativeTransformForItem(
    Child=(Type=Bone,Name="dyn_skirt_side_frnt_l_dynbase")
    Parent=(Type=Bone,Name="dyn_skirt_side_frnt_l_dynbase")
    Value=(Rotation=(X=0,Y=0,Z=0,W=1),Translation=(X=0,Y=0,Z=0),Scale3D=(X=1,Y=1,Z=1))
    bParentInitial=False
    bPropagateToChildren=True
    Weight=@n7.OutputParam
   ) {template="Set Relative Transform::Execute(in Child,in Parent,in bParentInitial,in Value,in Weight,in bPropagateToChildren)"}
   > &RigUnit_SetRelativeTransformForItem(
    Child=(Type=Bone,Name="dyn_skirt_side_r_base")
    Parent=(Type=Bone,Name="dyn_skirt_side_r_base")
    Value=(Rotation=(X=0.0002,Y=0.1475,Z=0.007,W=0.989),Translation=(X=0,Y=0,Z=0),Scale3D=(X=1,Y=1,Z=1))
    bParentInitial=False
    bPropagateToChildren=True
    Weight=@n7.OutputParam
   ) {template="Set Relative Transform::Execute(in Child,in Parent,in bParentInitial,in Value,in Weight,in bPropagateToChildren)"}
   &RigUnit_ParentConstraint(
    AdvancedSettings=(RotationOrderForFilter=XZY)
    Child=(Type=Bone,Name="dyn_arm_base_strings")
    Parents=((Item=(Type=Bone,Name="upperarm_l"),Weight=0.0046),(Item=(Type=Bone,Name="lowerarm_l"),Weight=0.2002),(Item=(Type=Bone,Name="lowerarm_twist_01_l"),Weight=0.2229),(Item=(Type=Bone,Name="lowerarm_twist_02_l"),Weight=0.5723))
    Weight=1
    ExecutePin=@RigVMFunction_Sequence.D
   )
   > n8 = &RigUnit_ParentConstraint(
    AdvancedSettings=(RotationOrderForFilter=XZY)
    Child=(Type=Bone,Name="dyn_base_belt")
    Parents=((Item=(Type=Bone,Name="spine_03"),Weight=0.0277),(Item=(Type=Bone,Name="spine_01"),Weight=0.0317),(Item=(Type=Bone,Name="spine_02"),Weight=0.0596),(Item=(Type=Bone,Name="spine_04"),Weight=0.128),(Item=(Type=Bone,Name="pelvis"),Weight=0.7531))
    Weight=1
   )
   RigVMFunction_ControlFlowBranch = &RigVMFunction_ControlFlowBranch(ExecuteContext=@n8.ExecutePin, Condition=RigVMVariableNode() {variable=cr}.Value)
   n9 = &RigVMFunction_MathDoubleRemap(SourceMaximum=1, SourceMinimum=0.3, TargetMaximum=1, TargetMinimum=0, Value=@n6.OutputParam)
   &RigUnit_SetRelativeTransformForItem(
    Child=(Type=Bone,Name="dyn_base_belt_loop_l")
    Parent=(Type=Bone,Name="dyn_base_belt_loop_l")
    Value=(Rotation=(X=-0.268,Y=0.295,Z=-0.1386,W=0.9066),Translation=(X=0,Y=0,Z=0),Scale3D=(X=1,Y=1,Z=1))
    bParentInitial=False
    bPropagateToChildren=True
    Weight=@n9.Result
    ExecutePin=@RigVMFunction_ControlFlowBranch.False
   ) {template="Set Relative Transform::Execute(in Child,in Parent,in bParentInitial,in Value,in Weight,in bPropagateToChildren)"}
   n10 = &RigVMFunction_MathDoubleRemap(SourceMaximum=0.6, SourceMinimum=0, TargetMaximum=1, TargetMinimum=0, Value=@n6.OutputParam)
   > &RigUnit_SetRelativeTransformForItem(
    Child=(Type=Bone,Name="dyn_base_belt_rope_l")
    Parent=(Type=Bone,Name="dyn_base_belt_rope_l")
    Value=(Rotation=(X=-0.1,Y=0.0829,Z=-0.0072,W=0.9915),Translation=(X=0,Y=0,Z=0),Scale3D=(X=1,Y=1,Z=1))
    bParentInitial=False
    bPropagateToChildren=True
    Weight=@n10.Result
   ) {template="Set Relative Transform::Execute(in Child,in Parent,in bParentInitial,in Value,in Weight,in bPropagateToChildren)"}
   > &RigUnit_SetRelativeTransformForItem(
    Child=(Type=Bone,Name="dyn_base_belt_rope_r")
    Parent=(Type=Bone,Name="dyn_base_belt_rope_r")
    Value=(Rotation=(X=0.022,Y=-0.1584,Z=0.1837,W=0.9699),Translation=(X=0,Y=0,Z=0),Scale3D=(X=1,Y=1,Z=1))
    bParentInitial=False
    bPropagateToChildren=True
    Weight=@n10.Result
   ) {template="Set Relative Transform::Execute(in Child,in Parent,in bParentInitial,in Value,in Weight,in bPropagateToChildren)"}
   > &RigUnit_SetRelativeTransformForItem(
    Child=(Type=Bone,Name="dyn_base_belt_loop_r")
    Parent=(Type=Bone,Name="dyn_base_belt_loop_r")
    Value=(Rotation=(X=-0.027,Y=-0.1142,Z=-0.1736,W=0.9778),Translation=(X=0,Y=0,Z=0),Scale3D=(X=1,Y=1,Z=1))
    bParentInitial=False
    bPropagateToChildren=True
    Weight=@n9.Result
   ) {template="Set Relative Transform::Execute(in Child,in Parent,in bParentInitial,in Value,in Weight,in bPropagateToChildren)"}
   > &RigUnit_SetRelativeTransformForItem(
    Child=(Type=Bone,Name="dyn_belt_loop_r2")
    Parent=(Type=Bone,Name="dyn_belt_loop_r2")
    Value=(Rotation=(X=-0.0001,Y=0,Z=0.2427,W=0.9701),Translation=(X=0,Y=0,Z=0),Scale3D=(X=1,Y=1,Z=1))
    bParentInitial=False
    bPropagateToChildren=True
    Weight=@n10.Result
   ) {template="Set Relative Transform::Execute(in Child,in Parent,in bParentInitial,in Value,in Weight,in bPropagateToChildren)"}
   > &RigUnit_SetRelativeTransformForItem(
    Child=(Type=Bone,Name="dyn_belt_rope_l1")
    Parent=(Type=Bone,Name="dyn_belt_rope_l1")
    Value=(Rotation=(X=-0.0086,Y=0.0004,Z=0.0877,W=0.9961),Translation=(X=0,Y=0,Z=0),Scale3D=(X=1,Y=1,Z=1))
    bParentInitial=False
    bPropagateToChildren=True
    Weight=@n6.OutputParam
   ) {template="Set Relative Transform::Execute(in Child,in Parent,in bParentInitial,in Value,in Weight,in bPropagateToChildren)"}
   > &RigUnit_SetRelativeTransformForItem(
    Child=(Type=Bone,Name="dyn_belt_loop_r1")
    Parent=(Type=Bone,Name="dyn_belt_loop_r1")
    Value=(Rotation=(X=-0.0974,Y=-0.0246,Z=0.1875,W=0.9771),Translation=(X=0,Y=0,Z=0),Scale3D=(X=1,Y=1,Z=1))
    bParentInitial=False
    bPropagateToChildren=True
    Weight=@n10.Result
   ) {template="Set Relative Transform::Execute(in Child,in Parent,in bParentInitial,in Value,in Weight,in bPropagateToChildren)"}
   > &RigUnit_SetRelativeTransformForItem(
    Child=(Type=Bone,Name="dyn_belt_rope_l2")
    Parent=(Type=Bone,Name="dyn_belt_rope_l2")
    Value=(Rotation=(X=-0.0004,Y=0.0002,Z=0.1183,W=0.993),Translation=(X=0,Y=0,Z=0),Scale3D=(X=1,Y=1,Z=1))
    bParentInitial=False
    bPropagateToChildren=True
    Weight=@n6.OutputParam
   ) {template="Set Relative Transform::Execute(in Child,in Parent,in bParentInitial,in Value,in Weight,in bPropagateToChildren)"}
   n11 = &RigVMFunction_MathDoubleRemap(SourceMaximum=0.9, SourceMinimum=0.1, TargetMaximum=1, TargetMinimum=0, Value=@n7.OutputParam)
   > &RigUnit_SetRelativeTransformForItem(
    Child=(Type=Bone,Name="dyn_base_belt_loop_r")
    Parent=(Type=Bone,Name="dyn_base_belt_loop_r")
    Value=(Rotation=(X=-0.0486,Y=-0.0728,Z=0.4086,W=0.9085),Translation=(X=0,Y=0,Z=0),Scale3D=(X=1,Y=1,Z=1))
    bParentInitial=False
    bPropagateToChildren=True
    Weight=@n11.Result
   ) {template="Set Relative Transform::Execute(in Child,in Parent,in bParentInitial,in Value,in Weight,in bPropagateToChildren)"}
   > &RigUnit_SetRelativeTransformForItem(
    Child=(Type=Bone,Name="dyn_base_belt_rope_r")
    Parent=(Type=Bone,Name="dyn_base_belt_rope_r")
    Value=(Rotation=(X=-0.0578,Y=0.181,Z=0.4037,W=0.8949),Translation=(X=0,Y=0,Z=0),Scale3D=(X=1,Y=1,Z=1))
    bParentInitial=False
    bPropagateToChildren=True
    Weight=@n11.Result
   ) {template="Set Relative Transform::Execute(in Child,in Parent,in bParentInitial,in Value,in Weight,in bPropagateToChildren)"}
   > &RigUnit_SetRelativeTransformForItem(
    Child=(Type=Bone,Name="dyn_base_belt_loop_l")
    Parent=(Type=Bone,Name="dyn_base_belt_loop_l")
    Value=(Rotation=(X=0.4575,Y=-0.092,Z=0.5066,W=0.725),Translation=(X=0,Y=0,Z=0),Scale3D=(X=1,Y=1,Z=1))
    bParentInitial=False
    bPropagateToChildren=True
    Weight=@n7.OutputParam
   ) {template="Set Relative Transform::Execute(in Child,in Parent,in bParentInitial,in Value,in Weight,in bPropagateToChildren)"}
   > &RigUnit_SetRelativeTransformForItem(
    Child=(Type=Bone,Name="dyn_base_belt_rope_l")
    Parent=(Type=Bone,Name="dyn_base_belt_rope_l")
    Value=(Rotation=(X=0.1855,Y=0.2535,Z=0.4859,W=0.8156),Translation=(X=0,Y=0,Z=0),Scale3D=(X=1,Y=1,Z=1))
    bParentInitial=False
    bPropagateToChildren=True
    Weight=@n7.OutputParam
   ) {template="Set Relative Transform::Execute(in Child,in Parent,in bParentInitial,in Value,in Weight,in bPropagateToChildren)"}
   > &RigUnit_SetRelativeTransformForItem(
    Child=(Type=Bone,Name="dyn_belt_loop_l1")
    Parent=(Type=Bone,Name="dyn_belt_loop_l1")
    Value=(Rotation=(X=-0.2745,Y=0.2818,Z=0.1283,W=0.9104),Translation=(X=0,Y=0,Z=0),Scale3D=(X=1,Y=1,Z=1))
    bParentInitial=False
    bPropagateToChildren=True
    Weight=@n6.OutputParam
   ) {template="Set Relative Transform::Execute(in Child,in Parent,in bParentInitial,in Value,in Weight,in bPropagateToChildren)"}
   > &RigUnit_SetRelativeTransformForItem(
    Child=(Type=Bone,Name="dyn_belt_loop_r1")
    Parent=(Type=Bone,Name="dyn_belt_loop_r1")
    Value=(Rotation=(X=0.0371,Y=0.0939,Z=0.2985,W=0.9491),Translation=(X=0,Y=0,Z=0),Scale3D=(X=1,Y=1,Z=1))
    bParentInitial=False
    bPropagateToChildren=True
    Weight=@n11.Result
   ) {template="Set Relative Transform::Execute(in Child,in Parent,in bParentInitial,in Value,in Weight,in bPropagateToChildren)"}
   > &RigUnit_SetRelativeTransformForItem(
    Child=(Type=Bone,Name="dyn_belt_rope_r1")
    Parent=(Type=Bone,Name="dyn_belt_rope_r1")
    Value=(Rotation=(X=-0.245,Y=0.1589,Z=0.2431,W=0.925),Translation=(X=0,Y=0,Z=0),Scale3D=(X=1,Y=1,Z=1))
    bParentInitial=False
    bPropagateToChildren=True
    Weight=@n11.Result
   ) {template="Set Relative Transform::Execute(in Child,in Parent,in bParentInitial,in Value,in Weight,in bPropagateToChildren)"}
   > &RigUnit_SetRelativeTransformForItem(
    Child=(Type=Bone,Name="dyn_belt_rope_l1")
    Parent=(Type=Bone,Name="dyn_belt_rope_l1")
    Value=(Rotation=(X=0.1053,Y=0.0244,Z=0.1428,W=0.9838),Translation=(X=0,Y=0,Z=0),Scale3D=(X=1,Y=1,Z=1))
    bParentInitial=False
    bPropagateToChildren=True
    Weight=@n6.OutputParam
   ) {template="Set Relative Transform::Execute(in Child,in Parent,in bParentInitial,in Value,in Weight,in bPropagateToChildren)"}
   > &RigUnit_SetRelativeTransformForItem(
    Child=(Type=Bone,Name="dyn_belt_rope_r2")
    Parent=(Type=Bone,Name="dyn_belt_rope_r2")
    Value=(Rotation=(X=0,Y=0,Z=0,W=1),Translation=(X=0,Y=0,Z=0),Scale3D=(X=1,Y=1,Z=1))
    bParentInitial=False
    bPropagateToChildren=True
    Weight=@n7.OutputParam
   ) {template="Set Relative Transform::Execute(in Child,in Parent,in bParentInitial,in Value,in Weight,in bPropagateToChildren)"}
   > &RigUnit_SetRelativeTransformForItem(
    Child=(Type=Bone,Name="dyn_belt_rope_l2")
    Parent=(Type=Bone,Name="dyn_belt_rope_l2")
    Value=(Rotation=(X=0,Y=0,Z=0,W=1),Translation=(X=0,Y=0,Z=0),Scale3D=(X=1,Y=1,Z=1))
    bParentInitial=False
    bPropagateToChildren=True
    Weight=@n6.OutputParam
   ) {template="Set Relative Transform::Execute(in Child,in Parent,in bParentInitial,in Value,in Weight,in bPropagateToChildren)"}
   > &RigUnit_SetRelativeTransformForItem(
    Child=(Type=Bone,Name="dyn_belt_loop_r2")
    Parent=(Type=Bone,Name="dyn_belt_loop_r2")
    Value=(Rotation=(X=0,Y=-0.0001,Z=0.0471,W=0.9989),Translation=(X=0,Y=0,Z=0),Scale3D=(X=1,Y=1,Z=1))
    bParentInitial=False
    bPropagateToChildren=True
    Weight=@n7.OutputParam
   ) {template="Set Relative Transform::Execute(in Child,in Parent,in bParentInitial,in Value,in Weight,in bPropagateToChildren)"}
   > &RigUnit_SetRelativeTransformForItem(
    Child=(Type=Bone,Name="dyn_belt_loop_l2")
    Parent=(Type=Bone,Name="dyn_belt_loop_l2")
    Value=(Rotation=(X=0.0152,Y=-0.0318,Z=0.4328,W=0.9008),Translation=(X=0,Y=0,Z=0),Scale3D=(X=1,Y=1,Z=1))
    bParentInitial=False
    bPropagateToChildren=True
    Weight=@n6.OutputParam
   ) {template="Set Relative Transform::Execute(in Child,in Parent,in bParentInitial,in Value,in Weight,in bPropagateToChildren)"}
   &RigUnit_SetRelativeTransformForItem(
    Child=(Type=Bone,Name="dyn_Sleeve_1_up_l")
    ExecutePin=()
    Parent=(Type=Bone,Name="dyn_Sleeve_1_up_l")
    Value=(Rotation=(X=0,Y=0,Z=-0.038,W=0.9993),Translation=(X=0,Y=0,Z=0),Scale3D=(X=1,Y=1,Z=1))
    bParentInitial=False
    bPropagateToChildren=True
    Weight=@n4.OutputParam
   ) {template="Set Relative Transform::Execute(in Child,in Parent,in bParentInitial,in Value,in Weight,in bPropagateToChildren)"}
   > &RigUnit_SetRelativeTransformForItem(
    Child=(Type=Bone,Name="dyn_Sleeve_1_up_r")
    Parent=(Type=Bone,Name="dyn_Sleeve_1_up_r")
    Value=(Rotation=(X=-0.0063,Y=-0.0339,Z=-0.0568,W=0.9978),Translation=(X=0,Y=0,Z=0),Scale3D=(X=1,Y=1,Z=1))
    bParentInitial=False
    bPropagateToChildren=True
    Weight=@n5.OutputParam
   ) {template="Set Relative Transform::Execute(in Child,in Parent,in bParentInitial,in Value,in Weight,in bPropagateToChildren)"}
   > &RigUnit_SetRelativeTransformForItem(
    Child=(Type=Bone,Name="dyn_Sleeve_1_down_l")
    Parent=(Type=Bone,Name="dyn_Sleeve_1_down_l")
    Value=(Rotation=(X=-0.0208,Y=-0.0664,Z=-0.0775,W=0.9946),Translation=(X=0.5203,Y=-0.1121,Z=0.4702),Scale3D=(X=1,Y=1,Z=1))
    bParentInitial=False
    bPropagateToChildren=True
    Weight=@n4.OutputParam
   ) {template="Set Relative Transform::Execute(in Child,in Parent,in bParentInitial,in Value,in Weight,in bPropagateToChildren)"}
   > &RigUnit_SetRelativeTransformForItem(
    Child=(Type=Bone,Name="dyn_Sleeve_1_down_r")
    Parent=(Type=Bone,Name="dyn_Sleeve_1_down_r")
    Value=(Rotation=(X=0.0105,Y=0.1281,Z=-0.0917,W=0.9875),Translation=(X=0,Y=0,Z=0),Scale3D=(X=1,Y=1,Z=1))
    bParentInitial=False
    bPropagateToChildren=True
    Weight=@n5.OutputParam
   ) {template="Set Relative Transform::Execute(in Child,in Parent,in bParentInitial,in Value,in Weight,in bPropagateToChildren)"}
   n12 = &RigVMFunction_MathDoubleRemap(SourceMaximum=0.87, SourceMinimum=0, TargetMaximum=1, TargetMinimum=0, Value=@n6.OutputParam)
   &RigUnit_SetRelativeTransformForItem(
    Child=(Type=Bone,Name="dyn_scarf_l_01")
    ExecutePin=()
    Parent=(Type=Bone,Name="dyn_scarf_l_01")
    Value=(Rotation=(X=-0.0076,Y=0.0231,Z=0.1071,W=0.994),Translation=(X=0,Y=0,Z=0),Scale3D=(X=1,Y=1,Z=1))
    bParentInitial=False
    bPropagateToChildren=True
    Weight=@n12.Result
   ) {template="Set Relative Transform::Execute(in Child,in Parent,in bParentInitial,in Value,in Weight,in bPropagateToChildren)"}
   > &RigUnit_SetRelativeTransformForItem(
    Child=(Type=Bone,Name="dyn_scarf_l_02")
    Parent=(Type=Bone,Name="dyn_scarf_l_02")
    Value=(Rotation=(X=-0.1661,Y=0.0792,Z=0.1614,W=0.9696),Translation=(X=-0.0838,Y=-0.008,Z=-0.093),Scale3D=(X=1,Y=1,Z=1))
    bParentInitial=False
    bPropagateToChildren=True
    Weight=@n12.Result
   ) {template="Set Relative Transform::Execute(in Child,in Parent,in bParentInitial,in Value,in Weight,in bPropagateToChildren)"}
   > &RigUnit_SetRelativeTransformForItem(
    Child=(Type=Bone,Name="dyn_scarf_r_01")
    Parent=(Type=Bone,Name="dyn_scarf_r_01")
    Value=(Rotation=(X=0.0001,Y=0,Z=0.0097,W=1),Translation=(X=0,Y=0,Z=0),Scale3D=(X=1,Y=1,Z=1))
    bParentInitial=False
    bPropagateToChildren=True
    Weight=@n12.Result
   ) {template="Set Relative Transform::Execute(in Child,in Parent,in bParentInitial,in Value,in Weight,in bPropagateToChildren)"}
   Remap = &RigVMFunction_MathDoubleRemap(SourceMaximum=0.8, SourceMinimum=0, TargetMaximum=1, TargetMinimum=0, Value=@n7.OutputParam)
   > &RigUnit_SetRelativeTransformForItem(
    Child=(Type=Bone,Name="dyn_scarf_r_01")
    Parent=(Type=Bone,Name="dyn_scarf_r_01")
    Value=(Rotation=(X=-0.0591,Y=-0.0419,Z=-0.0584,W=0.9957),Translation=(X=0,Y=0,Z=0),Scale3D=(X=1,Y=1,Z=1))
    bParentInitial=False
    bPropagateToChildren=True
    Weight=@Remap.Result
   ) {template="Set Relative Transform::Execute(in Child,in Parent,in bParentInitial,in Value,in Weight,in bPropagateToChildren)"}
   > &RigUnit_SetRelativeTransformForItem(
    Child=(Type=Bone,Name="dyn_scarf_r_02")
    Parent=(Type=Bone,Name="dyn_scarf_r_02")
    Value=(Rotation=(X=0.1165,Y=0.2284,Z=-0.3687,W=0.8935),Translation=(X=0,Y=0,Z=0),Scale3D=(X=1,Y=1,Z=1))
    bParentInitial=False
    bPropagateToChildren=True
    Weight=@Remap.Result
   ) {template="Set Relative Transform::Execute(in Child,in Parent,in bParentInitial,in Value,in Weight,in bPropagateToChildren)"}
   > &RigUnit_SetRelativeTransformForItem(
    Child=(Type=Bone,Name="dyn_scarf_l_01")
    Parent=(Type=Bone,Name="dyn_scarf_l_01")
    Value=(Rotation=(X=0.2636,Y=0.0417,Z=0.2157,W=0.9393),Translation=(X=0,Y=0,Z=0),Scale3D=(X=1,Y=1,Z=1))
    bParentInitial=False
    bPropagateToChildren=True
    Weight=@Remap.Result
   ) {template="Set Relative Transform::Execute(in Child,in Parent,in bParentInitial,in Value,in Weight,in bPropagateToChildren)"}
   > &RigUnit_SetRelativeTransformForItem(
    Child=(Type=Bone,Name="dyn_scarf_l_02")
    Parent=(Type=Bone,Name="dyn_scarf_l_02")
    Value=(Rotation=(X=-0.0175,Y=-0.0332,Z=-0.1277,W=0.9911),Translation=(X=0,Y=0,Z=0),Scale3D=(X=1,Y=1,Z=1))
    bParentInitial=False
    bPropagateToChildren=True
    Weight=@Remap.Result
   ) {template="Set Relative Transform::Execute(in Child,in Parent,in bParentInitial,in Value,in Weight,in bPropagateToChildren)"}
  ]}
]
hierarchy [
 root {loc=<0 0 0>, rot=<0 0 0 1>}
 pelvis {parent=root, loc=<0 -0.5301 98.0571>, rot=<0.0127 0.707 -0.0127 -0.707>}
 spine_01 {parent=pelvis, loc=<6.1208 -0.9328 0>, rot=<0 0 0.034 -0.9994>}
 spine_02 {parent=spine_01, loc=<6.5996 -0.2095 0>, rot=<0 0 0.0238 -0.9997>}
 spine_03 {parent=spine_02, loc=<6.4914 -0.1237 0>, rot=<0 0 0.0149 -0.9999>}
 spine_04 {parent=spine_03, loc=<6.9598 0.0794 0>, rot=<0 0 -0.0132 -0.9999>}
 spine_05 {parent=spine_04, loc=<6.9987 0 0>, rot=<0 0 -0.0015 -1>}
 attach_cape {parent=spine_05, loc=<9.0121 -11.7136 0.0005>, rot=<0.0616 -0.7044 -0.0616 -0.7044>}
 neck_01 {parent=spine_05, loc=<11.3138 0.1529 0>, rot=<0 0 -0.2003 -0.9797>}
 neck_02 {parent=neck_01, loc=<4.5627 0 0>, rot=<0 0 0.0504 -0.9987>}
 head {parent=neck_02, loc=<5.3914 0 0>, rot=<0 0 0.115 -0.9934>}
 attach_backpack {parent=spine_05, loc=<-0.1289 -13.4727 0.0006>, rot=<-0.0092 -0.707 0.0093 -0.707>}
 clavicle_l {parent=spine_05, loc=<5.4831 -0.1053 -3.2946>, rot=<0.0465 -0.7289 -0.009 -0.6829>}
 upperarm_l {parent=clavicle_l, loc=<9.8658 -1.2136 -0.4105>, rot=<0.0164 -0.3245 -0.0269 -0.9454>}
 lowerarm_l {parent=upperarm_l, loc=<26.465 -0.8345 0.259>, rot=<0.0036 0.0161 -0.2298 -0.9731>}
 hand_l {parent=lowerarm_l, loc=<22.7355 -0.3203 -0.6105>, rot=<-0.6784 -0.0223 0.0735 -0.7307>}
 thumb_01_l {parent=hand_l, loc=<1.3118 -0.3191 -2.459>, rot=<0.6546 -0.4042 0.0664 -0.6354>}
 thumb_02_l {parent=thumb_01_l, loc=<5.9336 -0.0448 0.0401>, rot=<-0.0037 0.0213 0.298 -0.9543>}
 thumb_03_l {parent=thumb_02_l, loc=<2.2223 0.3682 -0.0907>, rot=<0.0118 -0.013 0.0066 -0.9998>}
 index_metacarpal_l {parent=hand_l, loc=<1.6535 0.0183 -2.96>, rot=<0.0158 -0.0124 -0.0173 -0.9996>}
 index_01_l {parent=index_metacarpal_l, loc=<8.3008 0.0357 -0.0524>, rot=<-0.0044 0.0485 0.0924 -0.9945>}
 index_02_l {parent=index_01_l, loc=<3.3587 -0.3004 0.0452>, rot=<-0.002 0.0027 0.3225 -0.9466>}
 index_03_l {parent=index_02_l, loc=<2.5578 -0.123 -0.0252>, rot=<-0.002 0.0067 0.2558 -0.9667>}
 middle_metacarpal_l {parent=hand_l, loc=<2.1856 -0.1725 -0.5984>, rot=<0.006 0.0168 -0.0329 -0.9993>}
 middle_01_l {parent=middle_metacarpal_l, loc=<7.3275 -0.2447 0.0129>, rot=<-0.0208 0.0042 0.0858 -0.9961>}
 middle_02_l {parent=middle_01_l, loc=<4.0123 0.1401 0.1175>, rot=<-0.0218 0.0258 0.3512 -0.9357>}
 middle_03_l {parent=middle_02_l, loc=<2.6077 0.0534 0.1373>, rot=<0.0082 0.0188 0.1708 -0.9851>}
 ring_metacarpal_l {parent=hand_l, loc=<2.2074 -0.3625 1.3638>, rot=<-0.0284 0.0722 -0.0458 -0.9959>}
 ring_01_l {parent=ring_metacarpal_l, loc=<7.0295 -0.6197 -0.1446>, rot=<-0.0489 -0.0252 0.1892 -0.9804>}
 ring_02_l {parent=ring_01_l, loc=<3.6174 0.4647 -0.1027>, rot=<0 0 0.2806 -0.9598>}
 ring_03_l {parent=ring_02_l, loc=<2.4152 0.114 0.0199>, rot=<0 0 0.1189 -0.9929>}
 pinky_metacarpal_l {parent=hand_l, loc=<1.7616 -0.9753 3.3814>, rot=<-0.0397 0.0814 -0.069 -0.9935>}
 pinky_01_l {parent=pinky_metacarpal_l, loc=<7.1709 -0.4831 -0.1094>, rot=<-0.0654 -0.0035 0.2242 -0.9723>}
 pinky_02_l {parent=pinky_01_l, loc=<2.8274 0.0893 -0.1043>, rot=<0.0037 0.0011 0.2854 -0.9584>}
 pinky_03_l {parent=pinky_02_l, loc=<2.2046 0.0307 -0.0078>, rot=<-0.0002 0.0098 0.1445 -0.9895>}
 deform_wrist_dn_base_l {parent=hand_l, loc=<0 -1.5326 0>, rot=<0 0 0 -1>}
 deform_wrist_dn_l {parent=deform_wrist_dn_base_l, loc=<0 0 0>, rot=<0 0 0 1>}
 deform_wrist_up_base_l {parent=hand_l, loc=<0 2.3009 0>, rot=<0 0 0 -1>}
 deform_wrist_up_l {parent=deform_wrist_up_base_l, loc=<0 0 0>, rot=<0 0 0 1>}
 lowerarm_twist_02_l {parent=lowerarm_l, loc=<10.6707 0 0>, rot=<0 0 0 1>}
 lowerarm_twist_01_l {parent=lowerarm_l, loc=<18.316 0 0>, rot=<0 0 0 1>}
 deform_elbow_in_base_l {parent=lowerarm_l, loc=<0.1834 3.6062 -0.2572>, rot=<0 0 0 -1>}
 deform_elbow_in_l {parent=deform_elbow_in_base_l, loc=<0 0 0>, rot=<0 0 0 1>}
 deform_elbow_out_base_l {parent=lowerarm_l, loc=<0 -4 0>, rot=<0 0 0 1>}
 deform_elbow_out_l {parent=deform_elbow_out_base_l, loc=<0 0 0>, rot=<0 0 0 1>}
 dyn_arm_base_strings {parent=lowerarm_l, loc=<8.5182 2.1158 3.6278>, rot=<0.2895 -0.0281 0.0712 -0.9541>}
 dyn_arm_r_loop {parent=dyn_arm_base_strings, loc=<0.1959 0.6163 -0.2026>, rot=<0.4109 0.3105 -0.558 -0.6508>}
 dyn_arm_l_loop {parent=dyn_arm_base_strings, loc=<-0.2026 -0.5786 0.2263>, rot=<-0.1436 0.2796 0.8103 -0.4946>}
 dyn_arm_l_rope {parent=dyn_arm_base_strings, loc=<-0.1348 -0.4548 0.2213>, rot=<0.0056 0.0015 0.1436 -0.9896>}
 dyn_arm_r_rope {parent=dyn_arm_base_strings, loc=<0 0 0.2163>, rot=<0.2895 -0.0281 0.0712 -0.9541>}
 upperarm_twist_01_l {parent=upperarm_l, loc=<7.2135 -0.1134 0>, rot=<0 0 0 1>}
 dyn_Sleeve_1_up_l {parent=upperarm_twist_01_l, loc=<-5.3844 0.5492 3.702>, rot=<0 0.0626 0 -0.998>}
 dyn_Sleeve_1_down_l {parent=upperarm_twist_01_l, loc=<-1.2769 0.6806 -6.3462>, rot=<0 0 0.0521 -0.9986>}
 deform_upperarm_bk_base_l {parent=upperarm_l, loc=<-1.6448 -3.7411 1.2857>, rot=<0.1015 -0.013 0.2879 -0.9522>}
 deform_upperarm_bk_l {parent=deform_upperarm_bk_base_l, loc=<0 0 0>, rot=<0 0 0 1>}
 deform_upperarm_fr_base_l {parent=upperarm_l, loc=<3.7828 2.9436 -0.4842>, rot=<0.1015 -0.013 0.2879 -0.9522>}
 deform_upperarm_fr_l {parent=deform_upperarm_fr_base_l, loc=<0 0 0>, rot=<0 0 0 1>}
 deform_upperarm_up_base_l {parent=upperarm_l, loc=<-0.9979 1.6483 3.9385>, rot=<0.1072 -0.0162 0.204 -0.973>}
 deform_upperarm_up_l {parent=deform_upperarm_up_base_l, loc=<0 0 0>, rot=<0 0 0 1>}
 upperarm_twist_02_l {parent=upperarm_l, loc=<19.7998 -0.1134 0>, rot=<0 0 0 1>}
 clavicle_r {parent=spine_05, loc=<5.4831 -0.1053 3.2946>, rot=<0.7289 0.0465 -0.6829 0.009>}
 upperarm_r {parent=clavicle_r, loc=<-9.8658 1.2136 0.4105>, rot=<0.0164 -0.3245 -0.0269 -0.9454>}
 lowerarm_r {parent=upperarm_r, loc=<-26.465 0.8345 -0.259>, rot=<0.0036 0.0161 -0.2298 -0.9731>}
 hand_r {parent=lowerarm_r, loc=<-22.7355 0.3203 0.6106>, rot=<-0.6784 -0.0223 0.0735 -0.7307>}
 thumb_01_r {parent=hand_r, loc=<-1.3118 0.3191 2.459>, rot=<0.6546 -0.4042 0.0664 -0.6354>}
 thumb_02_r {parent=thumb_01_r, loc=<-5.9336 0.0448 -0.0401>, rot=<-0.0037 0.0213 0.298 -0.9543>}
 thumb_03_r {parent=thumb_02_r, loc=<-2.2223 -0.3682 0.0907>, rot=<0.0118 -0.013 0.0066 -0.9998>}
 index_metacarpal_r {parent=hand_r, loc=<-1.6535 -0.0183 2.96>, rot=<0.0158 -0.0124 -0.0173 -0.9996>}
 index_01_r {parent=index_metacarpal_r, loc=<-8.3008 -0.0357 0.0524>, rot=<-0.0044 0.0485 0.0924 -0.9945>}
 index_02_r {parent=index_01_r, loc=<-3.3587 0.3004 -0.0452>, rot=<-0.002 0.0027 0.3225 -0.9466>}
 index_03_r {parent=index_02_r, loc=<-2.5578 0.123 0.0252>, rot=<-0.002 0.0067 0.2558 -0.9667>}
 middle_metacarpal_r {parent=hand_r, loc=<-2.1856 0.1725 0.5984>, rot=<0.006 0.0168 -0.0329 -0.9993>}
 middle_01_r {parent=middle_metacarpal_r, loc=<-7.3275 0.2447 -0.0129>, rot=<-0.0208 0.0042 0.0858 -0.9961>}
 middle_02_r {parent=middle_01_r, loc=<-4.0123 -0.1401 -0.1175>, rot=<-0.0218 0.0258 0.3512 -0.9357>}
 middle_03_r {parent=middle_02_r, loc=<-2.6077 -0.0534 -0.1373>, rot=<0.0082 0.0188 0.1708 -0.9851>}
 ring_metacarpal_r {parent=hand_r, loc=<-2.2074 0.3625 -1.3638>, rot=<-0.0284 0.0722 -0.0458 -0.9959>}
 ring_01_r {parent=ring_metacarpal_r, loc=<-7.0295 0.6197 0.1446>, rot=<-0.0489 -0.0252 0.1892 -0.9804>}
 ring_02_r {parent=ring_01_r, loc=<-3.6174 -0.4647 0.1027>, rot=<0 0 0.2806 -0.9598>}
 ring_03_r {parent=ring_02_r, loc=<-2.4152 -0.114 -0.0199>, rot=<0 0 0.1189 -0.9929>}
 pinky_metacarpal_r {parent=hand_r, loc=<-1.7616 0.9753 -3.3814>, rot=<-0.0397 0.0814 -0.069 -0.9935>}
 pinky_01_r {parent=pinky_metacarpal_r, loc=<-7.1709 0.4831 0.1094>, rot=<-0.0654 -0.0035 0.2242 -0.9723>}
 pinky_02_r {parent=pinky_01_r, loc=<-2.8274 -0.0893 0.1043>, rot=<0.0037 0.0011 0.2854 -0.9584>}
 pinky_03_r {parent=pinky_02_r, loc=<-2.2046 -0.0307 0.0078>, rot=<-0.0002 0.0098 0.1445 -0.9895>}
 deform_wrist_dn_base_r {parent=hand_r, loc=<-0.0001 1.5328 0>, rot=<0 0 0 -1>}
 deform_wrist_dn_r {parent=deform_wrist_dn_base_r, loc=<0 0 0>, rot=<0 0 0 1>}
 deform_wrist_up_base_r {parent=hand_r, loc=<-0.0002 -2.3007 -0.0001>, rot=<0 0 0 -1>}
 deform_wrist_up_r {parent=deform_wrist_up_base_r, loc=<0 0 0>, rot=<0 0 0 1>}
 lowerarm_twist_02_r {parent=lowerarm_r, loc=<-10.6707 0 0>, rot=<0 0 0 -1>}
 lowerarm_twist_01_r {parent=lowerarm_r, loc=<-18.316 0 0>, rot=<0 0 0 -1>}
 deform_elbow_in_base_r {parent=lowerarm_r, loc=<-0.1834 -3.6062 0.2572>, rot=<0 0 0 -1>}
 deform_elbow_in_r {parent=deform_elbow_in_base_r, loc=<0 0 0>, rot=<0 0 0 1>}
 deform_elbow_out_base_r {parent=lowerarm_r, loc=<0 4 -0.0001>, rot=<0 0 0 1>}
 deform_elbow_out_r {parent=deform_elbow_out_base_r, loc=<0 0 0>, rot=<0 0 0 1>}
 upperarm_twist_01_r {parent=upperarm_r, loc=<-7.2135 0.1134 0>, rot=<0 0 0 -1>}
 dyn_Sleeve_1_up_r {parent=upperarm_twist_01_r, loc=<5.3844 -0.5492 -3.702>, rot=<-0.0626 0 0.998 0>}
 dyn_Sleeve_1_down_r {parent=upperarm_twist_01_r, loc=<1.277 -0.6806 6.3461>, rot=<0 0 0.9986 0.0521>}
 deform_upperarm_bk_base_r {parent=upperarm_r, loc=<1.9101 3.4032 -1.2378>, rot=<0.1705 0.2122 0.2863 -0.9187>}
 deform_upperarm_bk_r {parent=deform_upperarm_bk_base_r, loc=<0 0 0>, rot=<0 0 0 1>}
 deform_upperarm_fr_base_r {parent=upperarm_r, loc=<-3.5408 -3.0283 -0.3133>, rot=<0.1705 0.2122 0.2863 -0.9187>}
 deform_upperarm_fr_r {parent=deform_upperarm_fr_base_r, loc=<0 0 0>, rot=<0 0 0 1>}
 deform_upperarm_up_base_r {parent=upperarm_r, loc=<1.9053 -2.7134 -2.7398>, rot=<0.2049 0.1481 0.2048 -0.9456>}
 deform_upperarm_up_r {parent=deform_upperarm_up_base_r, loc=<0 0 0>, rot=<0 0 0 1>}
 upperarm_twist_02_r {parent=upperarm_r, loc=<-19.7998 0.1134 0>, rot=<0 0 0 -1>}
 base_back_collar {parent=spine_05, loc=<12.8215 -3.8484 0>, rot=<0 0 0.0872 -0.9962>}
 back_collar {parent=base_back_collar, loc=<0 0 0>, rot=<0 0 0 1>}
 base_front_collar_r {parent=spine_05, loc=<8.1744 4.3782 4.3781>, rot=<0.9096 0.2956 0.287 -0.0527>}
 front_collar_r {parent=base_front_collar_r, loc=<0 0 0>, rot=<0 0 0 1>}
 base_front_collar_l {parent=spine_05, loc=<8.1745 4.3782 -4.3781>, rot=<0.0527 -0.287 -0.2956 -0.9096>}
 front_collar_l {parent=base_front_collar_l, loc=<0 0 0>, rot=<0 0 0 1>}
 dyn_scarf_r_01 {parent=spine_05, loc=<9.2146 3.2824 7.3474>, rot=<0.0415 0.0175 0.9488 0.3125>}
 dyn_scarf_r_02 {parent=dyn_scarf_r_01, loc=<7.9971 0 0>, rot=<0.0015 -0.0024 -0.1586 -0.9873>}
 dyn_scarf_r_03 {parent=dyn_scarf_r_02, loc=<12.314 0 0>, rot=<0.0026 -0.0099 -0.1377 -0.9904>}
 dyn_scarf_r_04 {parent=dyn_scarf_r_03, loc=<8.918 0 0>, rot=<-0.001 0.0114 -0.0349 -0.9993>}
 dyn_scarf_r_05 {parent=dyn_scarf_r_04, loc=<9.2955 0 0>, rot=<0 0 0 -1>}
 dyn_scarf_l_01 {parent=spine_05, loc=<9.2145 3.2824 -7.3474>, rot=<-0.3145 -0.948 0.0432 -0.0215>}
 dyn_scarf_l_02 {parent=dyn_scarf_l_01, loc=<7.9969 0 0>, rot=<-0.027 0.0219 0.1571 -0.987>}
 dyn_scarf_l_03 {parent=dyn_scarf_l_02, loc=<12.3142 0 0>, rot=<-0.1204 0.031 0.1346 -0.9831>}
 dyn_scarf_l_04 {parent=dyn_scarf_l_03, loc=<8.918 0 0>, rot=<0.0011 0.0246 0.0272 -0.9993>}
 dyn_scarf_l_05 {parent=dyn_scarf_l_04, loc=<9.2958 0 0>, rot=<0 0 0 1>}
 side_collar_l_base {parent=spine_05, loc=<11.8412 0.4578 -5.9123>, rot=<0.1659 -0.2413 -0.2099 -0.9328>}
 side_collar_l {parent=side_collar_l_base, loc=<0 0 0>, rot=<0 0 0 1>}
 side_collar_r_base {parent=spine_05, loc=<11.8414 0.4578 5.9123>, rot=<0.9294 0.2076 0.2542 -0.1688>}
 side_collar_r {parent=side_collar_r_base, loc=<0 0 0>, rot=<0 0 0 1>}
 dyn_base_belt {parent=spine_02, loc=<-1.0156 11.1371 0.2794>, rot=<0.0399 -0.9992 0 -0.0003>}
 dyn_base_belt_loop_l {parent=dyn_base_belt, loc=<0.7861 0.0783 0>, rot=<0 0.3207 0 -0.9472>}
 dyn_belt_loop_l {parent=dyn_base_belt_loop_l, loc=<0 0 0>, rot=<0 0 0 1>}
 dyn_belt_loop_l2 {parent=dyn_belt_loop_l, loc=<3.8117 0 0>, rot=<0 0 0 1>}
 dyn_belt_loop_l1 {parent=dyn_belt_loop_l2, loc=<4.4974 0 0>, rot=<0 0 0 1>}
 dyn_base_belt_loop_r {parent=dyn_base_belt, loc=<0.5531 0.0783 -0.5161>, rot=<0 -0.3315 0 -0.9434>}
 dyn_belt_loop_r {parent=dyn_base_belt_loop_r, loc=<0 0 0>, rot=<0 0 0 1>}
 dyn_belt_loop_r2 {parent=dyn_belt_loop_r, loc=<3.8896 0 0>, rot=<0 0 0 1>}
 dyn_belt_loop_r1 {parent=dyn_belt_loop_r2, loc=<3.2133 0 0>, rot=<0 0 0 1>}
 dyn_base_belt_rope_r {parent=dyn_base_belt, loc=<0.3134 -0.839 0.1803>, rot=<0 -0.1393 0 -0.9902>}
 dyn_belt_rope_r {parent=dyn_base_belt_rope_r, loc=<0 0 0>, rot=<0 0 0 1>}
 dyn_belt_rope_r2 {parent=dyn_belt_rope_r, loc=<4.8145 0.2194 0>, rot=<0 0 0 1>}
 dyn_belt_rope_r1 {parent=dyn_belt_rope_r2, loc=<4.3448 0 0>, rot=<0 0 0 1>}
 dyn_base_belt_rope_l {parent=dyn_base_belt, loc=<0.6201 -0.7709 0.032>, rot=<-0.0002 -0.0561 0.0042 -0.9984>}
 dyn_belt_rope_l {parent=dyn_base_belt_rope_l, loc=<0 0 0>, rot=<0 0 0 1>}
 dyn_belt_rope_l2 {parent=dyn_belt_rope_l, loc=<4.4165 0.5941 0>, rot=<0 0 0 1>}
 dyn_belt_rope_l1 {parent=dyn_belt_rope_l2, loc=<4.2212 0 0>, rot=<0 0 0 1>}
 thigh_l {parent=pelvis, loc=<-1.0855 -0.2109 -8.7053>, rot=<0.0009 0.0355 0.0155 -0.9993>}
 calf_l {parent=thigh_l, loc=<-42.5284 -0.0019 -0.7297>, rot=<0.0006 -0.0161 -0.0391 -0.9991>}
 foot_l {parent=calf_l, loc=<-46.7401 -2.5014 -0.592>, rot=<0.0427 -0.0238 0.0353 -0.9982>}
 ball_l {parent=foot_l, loc=<-3.5658 10.8066 -0.1544>, rot=<0.0001 -0.0001 -0.7063 -0.7079>}
 calf_twist_02_l {parent=calf_l, loc=<-19.1546 -1.7943 -0.2638>, rot=<0.0237 0.0095 -0.0035 -0.9997>}
 calf_twist_01_l {parent=calf_l, loc=<-35.8483 -1.9197 -0.5794>, rot=<0.0237 0.0095 -0.0035 -0.9997>}
 deform_knee_in_base_l {parent=calf_l, loc=<0.8031 -5.0231 0.2456>, rot=<0 0 0 -1>}
 deform_knee_in_l {parent=deform_knee_in_base_l, loc=<0 0 0>, rot=<0 0 0 1>}
 deform_knee_out_base_l {parent=calf_l, loc=<0 4 0>, rot=<0 0 0 -1>}
 deform_knee_out_l {parent=deform_knee_out_base_l, loc=<0 0 0>, rot=<0 0 0 1>}
 thigh_twist_01_l {parent=thigh_l, loc=<-27.952 0 0>, rot=<0.0013 0 0 -1>}
 deform_groin_base_l {parent=thigh_l, loc=<0.4801 5.6776 -0.2486>, rot=<0 0 0 -1>}
 deform_groin_l {parent=deform_groin_base_l, loc=<0 0 0>, rot=<0 0 0 1>}
 dyn_skirt_side_frnt_l_dynbase {parent=thigh_l, loc=<-9.6008 7.8834 1.9259>, rot=<0.0546 -0.3999 0.9125 -0.0669>}
 dyn_skirt_side_frnt_l_01 {parent=dyn_skirt_side_frnt_l_dynbase, loc=<0 0 0>, rot=<0 0 0 1>}
 dyn_skirt_side_frnt_l_02 {parent=dyn_skirt_side_frnt_l_01, loc=<0 0 0>, rot=<0 0 0 1>}
 thigh_r {parent=pelvis, loc=<-1.0855 -0.2109 8.7053>, rot=<0.0355 -0.0009 0.9993 0.0155>}
 calf_r {parent=thigh_r, loc=<42.5284 0.0019 0.7297>, rot=<0.0006 -0.0161 -0.0391 -0.9991>}
 foot_r {parent=calf_r, loc=<46.7401 2.5014 0.592>, rot=<0.0427 -0.0238 0.0353 -0.9982>}
 ball_r {parent=foot_r, loc=<3.5658 -10.8066 0.1544>, rot=<0.0001 -0.0001 -0.7063 -0.7079>}
 calf_twist_02_r {parent=calf_r, loc=<19.1546 1.7943 0.2638>, rot=<0.0237 0.0095 -0.0035 -0.9997>}
 calf_twist_01_r {parent=calf_r, loc=<35.8483 1.9197 0.5794>, rot=<0.0237 0.0095 -0.0035 -0.9997>}
 deform_knee_in_base_r {parent=calf_r, loc=<-0.8031 5.0231 -0.2457>, rot=<0 0 0 -1>}
 deform_knee_in_r {parent=deform_knee_in_base_r, loc=<0 0 0>, rot=<0 0 0 1>}
 deform_knee_out_base_r {parent=calf_r, loc=<0 -4 0>, rot=<0 0 0 -1>}
 deform_knee_out_r {parent=deform_knee_out_base_r, loc=<0 0 0>, rot=<0 0 0 1>}
 thigh_twist_01_r {parent=thigh_r, loc=<27.952 0 0>, rot=<0.0013 0 0 -1>}
 deform_groin_base_r {parent=thigh_r, loc=<-0.4801 -5.6776 0.2486>, rot=<0 0 0 -1>}
 deform_groin_r {parent=deform_groin_base_r, loc=<0 0 0>, rot=<0 0 0 1>}
 dyn_skirt_side_frnt_r_dynbase {parent=thigh_r, loc=<10.112 -8.2512 -0.9548>, rot=<0.9011 0.0106 0.0629 -0.4289>}
 dyn_skirt_side_frnt_r_01 {parent=dyn_skirt_side_frnt_r_dynbase, loc=<0 0 0>, rot=<0 0 0 1>}
 dyn_skirt_side_frnt_r_base {parent=pelvis, loc=<5.6103 7.8194 5.1796>, rot=<0.0119 -0.8988 0.4264 0.1015>}
 dyn_skirt_side_r_base {parent=pelvis, loc=<7.1354 0.3289 10.9584>, rot=<0.0234 -0.9921 -0.0163 0.1221>}
 dyn_skirt_side_r_parent {parent=dyn_skirt_side_r_base, loc=<0 0 0>, rot=<0 0 0 1>}
 dyn_skirt_side_r_01 {parent=dyn_skirt_side_r_parent, loc=<16.4926 -0.5849 0.7226>, rot=<0 0 0 1>}
 dyn_skirt_side_bk_r_base {parent=pelvis, loc=<6.5526 -8.1461 5.9217>, rot=<-0.0025 -0.8939 -0.4314 0.1215>}
 dyn_skirt_side_bk_r_parent {parent=dyn_skirt_side_bk_r_base, loc=<0 0 0>, rot=<0 0 0 1>}
 dyn_skirt_side_bk_r_01 {parent=dyn_skirt_side_bk_r_parent, loc=<16.9212 -0.2778 -0.3085>, rot=<0.0128 -0.0027 0.0172 -0.9998>}
 dyn_skirt_side_bk_l_base {parent=pelvis, loc=<6.5597 -8.169 -5.8775>, rot=<-0.1428 0.481 0.865 0.0111>}
 dyn_skirt_side_bk_l_parent {parent=dyn_skirt_side_bk_l_base, loc=<0 0 0>, rot=<0 0 0 1>}
 dyn_skirt_side_bk_l_01 {parent=dyn_skirt_side_bk_l_parent, loc=<16.8885 0.4367 0.4611>, rot=<0.0221 0.0716 -0.0028 -0.9972>}
 dyn_skirt_side_l_base {parent=pelvis, loc=<7.1912 0.3658 -10.9406>, rot=<-0.1371 0.0102 0.9897 -0.0391>}
 dyn_skirt_side_l_parent {parent=dyn_skirt_side_l_base, loc=<0 0 0>, rot=<0 0 0 1>}
 dyn_skirt_side_l_01 {parent=dyn_skirt_side_l_parent, loc=<16.4612 0.1114 1.1803>, rot=<0 0 0 1>}
 dyn_skirt_side_frnt_l_base {parent=pelvis, loc=<6.0985 8.0095 -5.1411>, rot=<0.0161 -0.3973 0.9151 -0.0668>}
]
variables [cr {type=bool, default=False}]
