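# Physics asset for the scarf of the female medium FelineWarrior player body, using capsule shapes throughout.
# Two five-bone scarf chains, left and right, hang from spine_05 through angular-limited constraints that are tightest near the root and looser further down.
# Capsules on the pelvis, spine, thighs and two front-side skirt bones have no constraints, apart from the spine_05 anchor, and appear to serve as colliders for the scarf.
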PhysicsAsset /Game/Characters/Player/Female/Medium/Bodies/F_MED_FelineWarrior/Meshes/F_MED_FelineWarrior__scarf_Physics_03
bodies {
 dyn_scarf_l_01 [{
   shapeType=Capsule
   length=8.8
   radius=4.7
   shapeName=dyn_scarf_l_01_capsule
   center=<6.3951 0.6567 -0.7216>
   rotation=<rot 10.5025 -1.6615 -9.046>
  }]
 dyn_scarf_l_02 [{
   shapeType=Capsule
   length=8.3
   radius=5.2
   shapeName=dyn_scarf_l_02_capsule
   center=<7.0613 0.2004 -0.6118>
   rotation=<rot 5.5337 -0.868 -8.9892>
  }]
 dyn_scarf_l_03 [{
   shapeType=Capsule
   length=4.0732
   radius=4.1504
   shapeName=dyn_scarf_l_03_capsule
   center=<4.7681 0.214 0.0041>
   rotation=<rot -0.0002 0.0005 3.5611>
  }]
 dyn_scarf_l_04 [{
   shapeType=Capsule
   length=5.6202
   radius=5.2325
   shapeName=dyn_scarf_l_04_capsule
   center=<4.9738 0.1974 0.0109>
   rotation=<rot -0.0002 0.0004 -8.8932>
  }]
 dyn_scarf_l_05 [{
   shapeType=Capsule
   length=8.9327
   radius=4.9311
   shapeName=dyn_scarf_l_05_capsule
   center=<4.9738 0.1974 0.0109>
   rotation=<rot -0.0002 0.0004 -8.8932>
  }]
 dyn_scarf_r_01 [{
   shapeType=Capsule
   length=8.8
   radius=4.7
   shapeName=dyn_scarf_r_01_capsule
   center=<7.2268 -1.1558 0.82>
   rotation=<rot 10.2479 6.1268 31.1082>
  }]
 dyn_scarf_r_02 [{
   shapeType=Capsule
   length=8.3
   radius=5.2
   shapeName=dyn_scarf_r_02_capsule
   center=<7.818 -0.6746 0.7943>
   rotation=<rot 3.9488 2.3348 30.6388>
  }]
 dyn_scarf_r_03 [{
   shapeType=Capsule
   length=7.8
   radius=5.7
   shapeName=dyn_scarf_r_03_capsule
   center=<5.727 -0.6201 0.5319>
   rotation=<rot -0.8513 -0.5061 30.5622>
  }]
 dyn_scarf_r_04 [{shapeType=Capsule, length=9, radius=5.5, shapeName=dyn_scarf_r_04_capsule, center=<5.9379 -0.4499 0.7899>, rotation=<rot 0 -0.0006 30.5583>}]
 dyn_scarf_r_05 [{
   shapeType=Capsule
   length=10.7336
   radius=4.3664
   shapeName=dyn_scarf_r_05_capsule
   center=<5.9379 -2.3616 1.9186>
   rotation=<rot 0 -0.0006 30.5583>
  }]
 dyn_skirt_side_frnt_l_02 [{
   shapeType=Capsule
   length=10.4077
   radius=6.4647
   shapeName=dyn_skirt_side_frnt_l_02_capsule
   center=<13.1013 -0.6415 1.2644>
   rotation=<rot 79.4389 27.226 117.3121>
  }]
 dyn_skirt_side_frnt_r_01 [{
   shapeType=Capsule
   length=12.4889
   radius=9.1834
   shapeName=dyn_skirt_side_frnt_r_01_capsule
   center=<6.6734 -5.8775 8.3542>
   rotation=<rot 6.8654 5.176 70.4526>
  }]
 pelvis [{shapeType=Capsule, length=17.4833, radius=10, shapeName=pelvis_capsule, center=<0 0.1476 0>}]
 spine_01 [
  {shapeType=Capsule, length=53.3386, radius=10, shapeName=spine_01_capsule, center=<0 -0.5244 -13.3707>}
  {shapeType=Capsule, length=15, radius=15, shapeName=spine_01_capsule, center=<7.7998 10.1143 -33.6035>, rotation=<rot 88.6685 179.9996 180>}
  {shapeType=Capsule, length=8.0272, radius=10.6668, shapeName=spine_01_capsule, center=<3.4274 8.7479 -7.6227>}
 ]
 spine_02 [
  {shapeType=Capsule, length=14.3675, radius=10, shapeName=spine_02_capsule, center=<0 -1.1684 0>}
  {
   shapeType=Capsule
   length=0
   radius=8.8156
   shapeName=spine_02_capsule
   center=<5.8544 3.3723 -2.8454>
   rotation=<rot -1.0819 0.1555 177.343>
  }
 ]
 spine_03 [{shapeType=Capsule, length=49.7885, radius=10.5, shapeName=spine_03_capsule, center=<-0.4309 -3.1701 -13.8483>}]
 spine_04 [{shapeType=Capsule, length=15, radius=15, shapeName=spine_04_capsule, center=<-1.8769 -14.3951 0.4786>, rotation=<rot 0 -0.0003 91.9041>}]
 spine_05 [{shapeType=Capsule, length=9.3496, radius=11.9, shapeName=spine_05_capsule, center=<-2.2782 -6.6968 0>}]
 thigh_l [{
   shapeType=Capsule
   length=33.0861
   radius=9.5232
   shapeName=thigh_l_capsule
   center=<-26.6333 -0.7628 -2.0675>
   rotation=<rot -87.5866 -28.9828 34.0646>
  }]
 thigh_r [{
   shapeType=Capsule
   length=33.8008
   radius=8.8086
   shapeName=thigh_r_capsule
   center=<23.0281 1.1484 0.1522>
   rotation=<rot -89.6322 -33.907 38.5855>
  }]
}
constraints [
 {bone1Name=dyn_scarf_l_01, bone2Name=spine_05, swing1LimitDegrees=5, swing2LimitDegrees=15, twistLimitDegrees=2}
 {bone1Name=dyn_scarf_l_02, bone2Name=dyn_scarf_l_01, swing1LimitDegrees=10, swing2LimitDegrees=10, twistLimitDegrees=2}
 {bone1Name=dyn_scarf_l_03, bone2Name=dyn_scarf_l_02, swing1LimitDegrees=20, swing2LimitDegrees=30, twistLimitDegrees=20}
 {bone1Name=dyn_scarf_l_04, bone2Name=dyn_scarf_l_03, swing1LimitDegrees=45, swing2LimitDegrees=15, twistLimitDegrees=15}
 {bone1Name=dyn_scarf_l_05, bone2Name=dyn_scarf_l_04, swing1LimitDegrees=15, swing2LimitDegrees=15, twistLimitDegrees=15}
 {bone1Name=dyn_scarf_r_01, bone2Name=spine_05, swing1LimitDegrees=5, swing2LimitDegrees=15, twistLimitDegrees=2}
 {bone1Name=dyn_scarf_r_02, bone2Name=dyn_scarf_r_01, swing1LimitDegrees=10, swing2LimitDegrees=15, twistLimitDegrees=2}
 {bone1Name=dyn_scarf_r_03, bone2Name=dyn_scarf_r_02, swing1LimitDegrees=40, swing2LimitDegrees=15, twistLimitDegrees=10}
 {bone1Name=dyn_scarf_r_04, bone2Name=dyn_scarf_r_03, swing1LimitDegrees=45, swing2LimitDegrees=15, twistLimitDegrees=15}
 {bone1Name=dyn_scarf_r_05, bone2Name=dyn_scarf_r_04, swing1LimitDegrees=45, swing2LimitDegrees=15, twistLimitDegrees=15}
]
