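# Physics asset for the scarf on the medium female FelineWarrior player body, apparently the frontend variant by its name.
# Left and right five-bone scarf capsule chains hang from spine_05, linked by tight swing and twist limits that loosen most in swing1 at the third bone.
# Capsules on the pelvis, spine and thighs form the body volume; two belt-loop capsules have no constraints.
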PhysicsAsset /Game/Characters/Player/Female/Medium/Bodies/F_MED_FelineWarrior/Meshes/F_MED_FelineWarrior__scarf_Physics_frontend_01
bodies {
 dyn_belt_loop_l1 [{
   shapeType=Capsule
   length=5.9229
   radius=2.9637
   shapeName=dyn_belt_loop_l1_capsule
   center=<6.0199 -4.9803 -2.2758>
   rotation=<rot 0 0.0002 -33.6683>
  }]
 dyn_belt_loop_r1 [{
   shapeType=Capsule
   length=5.9229
   radius=2.9637
   shapeName=dyn_belt_loop_r1_capsule
   center=<4.7592 -5.2926 2.5403>
   rotation=<rot 49.7963 20.1374 43.2742>
  }]
 dyn_scarf_l_01 [{
   shapeType=Capsule
   length=8.6672
   radius=4.8328
   shapeName=dyn_scarf_l_01_capsule
   center=<6.4606 0.8021 -0.6864>
   rotation=<rot 5.9256 -25.586 -12.5164>
  }]
 dyn_scarf_l_02 [{
   shapeType=Capsule
   length=8.3
   radius=5.2
   shapeName=dyn_scarf_l_02_capsule
   center=<7.0613 0.2004 -0.6118>
   rotation=<rot 5.5337 -0.868 -8.9352>
  }]
 dyn_scarf_l_03 [{
   shapeType=Capsule
   length=8
   radius=5.5
   shapeName=dyn_scarf_l_03_capsule
   center=<4.7681 0.214 0.0041>
   rotation=<rot -0.0002 0.0004 -8.8932>
  }]
 dyn_scarf_l_04 [{
   shapeType=Capsule
   length=10.2662
   radius=4.2338
   shapeName=dyn_scarf_l_04_capsule
   center=<4.9738 0.1974 0.0109>
   rotation=<rot -0.0002 0.0004 -8.8932>
  }]
 dyn_scarf_l_05 [{
   shapeType=Capsule
   length=11.1662
   radius=3.9338
   shapeName=dyn_scarf_l_05_capsule
   center=<4.9738 0.1974 0.0109>
   rotation=<rot -0.0002 0.0004 -8.8932>
  }]
 dyn_scarf_r_01 [{
   shapeType=Capsule
   length=8.8
   radius=4.7
   shapeName=dyn_scarf_r_01_capsule
   center=<7.2268 -1.1558 0.82>
   rotation=<rot 10.2479 6.1268 31.1082>
  }]
 dyn_scarf_r_02 [{
   shapeType=Capsule
   length=8.3
   radius=5.2
   shapeName=dyn_scarf_r_02_capsule
   center=<7.818 -0.6746 0.7943>
   rotation=<rot 3.9488 2.3348 30.6388>
  }]
 dyn_scarf_r_03 [{
   shapeType=Capsule
   length=9.3
   radius=4.2
   shapeName=dyn_scarf_r_03_capsule
   center=<5.727 -0.6201 0.5319>
   rotation=<rot -0.8513 -0.5061 30.5622>
  }]
 dyn_scarf_r_04 [{
   shapeType=Capsule
   length=10.7662
   radius=3.7338
   shapeName=dyn_scarf_r_04_capsule
   center=<5.9379 -0.4499 0.7899>
   rotation=<rot 0 -0.0006 30.5583>
  }]
 dyn_scarf_r_05 [{
   shapeType=Capsule
   length=12.1662
   radius=2.9338
   shapeName=dyn_scarf_r_05_capsule
   center=<5.9379 -0.4499 0.7899>
   rotation=<rot 0 -0.0006 30.5583>
  }]
 pelvis [{shapeType=Capsule, length=18.4833, radius=9, shapeName=pelvis_capsule, center=<0 -4.1147 0>}]
 spine_01 [{shapeType=Capsule, length=17.0309, radius=7.817, shapeName=spine_01_capsule, center=<0 -1.8817 0>}]
 spine_02 [
  {shapeType=Capsule, length=14.3675, radius=10, shapeName=spine_02_capsule, center=<0 -3.5452 0>}
  {
   shapeType=Capsule
   length=6.2747
   radius=2.7617
   shapeName=spine_02_capsule
   center=<-1.1838 7.8723 0.3961>
   rotation=<rot 0 -0.0001 91.7194>
  }
 ]
 spine_03 [{shapeType=Capsule, length=14.0659, radius=12.0037, shapeName=spine_03_capsule, center=<-0.4309 -6.4817 0>}]
 spine_04 [
  {shapeType=Capsule, length=15, radius=15, shapeName=spine_04_capsule, center=<-1.8769 -14.3951 0.4786>, rotation=<rot 0 -0.0003 91.9041>}
  {shapeType=Capsule, length=15, radius=15, shapeName=spine_04_capsule, center=<0 -10.404 0>}
 ]
 spine_05 [{shapeType=Capsule, length=9.2712, radius=11.9785, shapeName=spine_05_capsule, center=<-3.9164 -6.1073 0>}]
 thigh_l [{
   shapeType=Capsule
   length=34.977
   radius=7.6324
   shapeName=thigh_l_capsule
   center=<-25.6791 0 -1.072>
   rotation=<rot -89.8326 -179.9943 179.9941>
  }]
 thigh_r [{
   shapeType=Capsule
   length=34.977
   radius=7.6324
   shapeName=thigh_r_capsule
   center=<25.3101 0.3455 -1.2291>
   rotation=<rot -85.5327 -171.5711 177.0663>
  }]
}
constraints [
 {bone1Name=dyn_scarf_l_01, bone2Name=spine_05, swing1LimitDegrees=2, swing2LimitDegrees=2, twistLimitDegrees=2}
 {bone1Name=dyn_scarf_l_02, bone2Name=dyn_scarf_l_01, swing1LimitDegrees=5, swing2LimitDegrees=2, twistLimitDegrees=2}
 {bone1Name=dyn_scarf_l_03, bone2Name=dyn_scarf_l_02, swing1LimitDegrees=10, swing2LimitDegrees=2, twistLimitDegrees=2}
 {bone1Name=dyn_scarf_l_04, bone2Name=dyn_scarf_l_03, swing1LimitDegrees=5, swing2LimitDegrees=2, twistLimitDegrees=2}
 {bone1Name=dyn_scarf_l_05, bone2Name=dyn_scarf_l_04, swing1LimitDegrees=5, swing2LimitDegrees=2, twistLimitDegrees=2}
 {bone1Name=dyn_scarf_r_01, bone2Name=spine_05, swing1LimitDegrees=2, swing2LimitDegrees=2, twistLimitDegrees=2}
 {bone1Name=dyn_scarf_r_02, bone2Name=dyn_scarf_r_01, swing1LimitDegrees=5, swing2LimitDegrees=2, twistLimitDegrees=2}
 {bone1Name=dyn_scarf_r_03, bone2Name=dyn_scarf_r_02, swing1LimitDegrees=10, swing2LimitDegrees=2, twistLimitDegrees=2}
 {bone1Name=dyn_scarf_r_04, bone2Name=dyn_scarf_r_03, swing1LimitDegrees=5, swing2LimitDegrees=2, twistLimitDegrees=2}
 {bone1Name=dyn_scarf_r_05, bone2Name=dyn_scarf_r_04, swing1LimitDegrees=5, swing2LimitDegrees=2, twistLimitDegrees=2}
]
