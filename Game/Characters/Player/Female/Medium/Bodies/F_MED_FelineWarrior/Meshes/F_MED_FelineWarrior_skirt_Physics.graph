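# Physics asset for the Feline Warrior medium female player body's skirt, defining collision and joint limits for its dynamic bones.
# Back and side strips chain base, parent and tip capsules, while front strips constrain tips to dynbase bodies, all with tight angular limits.
# Pelvis and thigh capsules provide body collision shapes and have no constraints.
# Base bodies have no constraints to the pelvis, and front base bodies have none at all.
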PhysicsAsset /Game/Characters/Player/Female/Medium/Bodies/F_MED_FelineWarrior/Meshes/F_MED_FelineWarrior_skirt_Physics
bodies {
 dyn_skirt_side_bk_l_01 [{
   shapeType=Capsule
   length=22.3
   radius=8.6
   shapeName=dyn_skirt_side_bk_l_01_capsule
   center=<6.8097 4.7199 -6.5618>
   rotation=<rot 1.2312 8.9769 68.6658>
  }]
 dyn_skirt_side_bk_l_base [{shapeType=Capsule, length=0.7, radius=2, shapeName=dyn_skirt_side_bk_l_base_capsule}]
 dyn_skirt_side_bk_l_parent [{
   shapeType=Capsule
   length=16
   radius=8.6
   shapeName=dyn_skirt_side_bk_l_parent_capsule
   center=<6.5426 6.696 -8.3776>
   rotation=<rot 2.3738 12.2445 70.4485>
  }]
 dyn_skirt_side_bk_r_01 [{
   shapeType=Capsule
   length=21.2
   radius=8.6
   shapeName=dyn_skirt_side_bk_r_01_capsule
   center=<6.9679 -1.8425 -6.0607>
   rotation=<rot 10.3923 -2.8769 111.7521>
  }]
 dyn_skirt_side_bk_r_base [{shapeType=Capsule, length=0.7, radius=2, shapeName=dyn_skirt_side_bk_r_base_capsule}]
 dyn_skirt_side_bk_r_parent [{
   shapeType=Capsule
   length=13.8
   radius=8.6
   shapeName=dyn_skirt_side_bk_r_parent_capsule
   center=<7.9419 1.3297 -11.7692>
   rotation=<rot 11.6157 -3.1712 107.1268>
  }]
 dyn_skirt_side_frnt_l_01 [
  {
   shapeType=Capsule
   length=23
   radius=7.6
   shapeName=dyn_skirt_side_frnt_l_01_capsule
   center=<6.4858 -3.0815 -8.5366>
   rotation=<rot -3.5537 -6.7595 115.7414>
  }
  {
   shapeType=Capsule
   length=17.9
   radius=6
   shapeName=dyn_skirt_side_frnt_l_01_capsule
   center=<14.55 -5.3539 -6.9359>
   rotation=<rot -11.769 12.4759 -67.2276>
  }
 ]
 dyn_skirt_side_frnt_l_base [{shapeType=Capsule, length=0.7, radius=2, shapeName=dyn_skirt_side_frnt_l_base_capsule}]
 dyn_skirt_side_frnt_l_dynbase [{shapeType=Capsule, length=4.3, radius=1, shapeName=dyn_skirt_side_frnt_l_dynbase_capsule}]
 dyn_skirt_side_frnt_r_01 [{
   shapeType=Capsule
   length=18.3
   radius=8
   shapeName=dyn_skirt_side_frnt_r_01_capsule
   center=<7.0652 2.1843 -6.8575>
   rotation=<rot 2.5282 6.1486 61.9134>
  }]
 dyn_skirt_side_frnt_r_base [{shapeType=Capsule, length=0.7, radius=2, shapeName=dyn_skirt_side_frnt_r_base_capsule}]
 dyn_skirt_side_frnt_r_dynbase [{shapeType=Capsule, length=4.3, radius=1, shapeName=dyn_skirt_side_frnt_r_dynbase_capsule}]
 dyn_skirt_side_l_01 [{
   shapeType=Capsule
   length=13.8
   radius=8.6
   shapeName=dyn_skirt_side_r_01_capsule
   center=<9.3561 -0.1972 -9.0698>
   rotation=<rot 10.0858 -6.131 -90.4707>
  }]
 dyn_skirt_side_l_base [{shapeType=Capsule, length=0.7, radius=2, shapeName=dyn_skirt_side_l_base_capsule}]
 dyn_skirt_side_l_parent [{
   shapeType=Capsule
   length=15.8542
   radius=6.5458
   shapeName=dyn_skirt_side_r_01_capsule
   center=<9.0179 -0.1769 -7.1676>
   rotation=<rot 10.0898 -5.5719 -90.3728>
  }]
 dyn_skirt_side_r_01 [{
   shapeType=Capsule
   length=13.8
   radius=8.6
   shapeName=dyn_skirt_side_r_01_capsule
   center=<7.9373 1.5245 -9.3628>
   rotation=<rot 10.0451 2.4035 91.0236>
  }]
 dyn_skirt_side_r_base [{shapeType=Capsule, length=0.7, radius=2, shapeName=dyn_skirt_side_r_base_capsule}]
 dyn_skirt_side_r_parent [{
   shapeType=Capsule
   length=16.1
   radius=6.3
   shapeName=dyn_skirt_side_r_parent_capsule
   center=<8.5676 1.5582 -10.4336>
   rotation=<rot 11.8917 2.398 90.0001>
  }]
 pelvis [
  {shapeType=Capsule, length=9.3038, radius=9.906, center=<1.6959 -0.8879 0>}
  {shapeType=Capsule, length=8.6, radius=11.2, shapeName=pelvis_capsule, center=<-5.1854 -1.5559 0>}
 ]
 thigh_l [{shapeType=Capsule, length=29.3692, radius=8.4475, center=<-26.4705 -0.2363 0.913>, rotation=<rot 89.9131 1.3175 0.0042>}]
 thigh_r [{shapeType=Capsule, length=29.5693, radius=8.2475, center=<25.0061 0.0216 -0.5103>, rotation=<rot 88.9334 4.26 -176.8992>}]
}
constraints [
 {bone1Name=dyn_skirt_side_bk_l_parent, bone2Name=dyn_skirt_side_bk_l_base, swing1LimitDegrees=3, swing2LimitDegrees=15, twistLimitDegrees=2}
 {bone1Name=dyn_skirt_side_bk_l_01, bone2Name=dyn_skirt_side_bk_l_parent, swing1LimitDegrees=3, swing2LimitDegrees=30, twistLimitDegrees=2}
 {bone1Name=dyn_skirt_side_bk_r_parent, bone2Name=dyn_skirt_side_bk_r_base, swing1LimitDegrees=3, swing2LimitDegrees=15, twistLimitDegrees=2}
 {bone1Name=dyn_skirt_side_bk_r_01, bone2Name=dyn_skirt_side_bk_r_parent, swing1LimitDegrees=3, swing2LimitDegrees=30, twistLimitDegrees=2}
 {bone1Name=dyn_skirt_side_r_parent, bone2Name=dyn_skirt_side_r_base, swing1LimitDegrees=10, swing2LimitDegrees=10, twistLimitDegrees=10}
 {bone1Name=dyn_skirt_side_r_01, bone2Name=dyn_skirt_side_r_parent, swing1LimitDegrees=2, swing2LimitDegrees=15, twistLimitDegrees=2}
 {
  bone1Name=dyn_skirt_side_frnt_l_01
  bone2Name=dyn_skirt_side_frnt_l_dynbase
  swing1LimitDegrees=15
  swing2LimitDegrees=10
  twistLimitDegrees=10
 }
 {
  bone1Name=dyn_skirt_side_frnt_r_01
  bone2Name=dyn_skirt_side_frnt_r_dynbase
  swing1LimitDegrees=15
  swing2LimitDegrees=10
  twistLimitDegrees=3
 }
 {bone1Name=dyn_skirt_side_l_01, bone2Name=dyn_skirt_side_l_parent, swing1LimitDegrees=2, swing2LimitDegrees=15, twistLimitDegrees=3}
 {bone1Name=dyn_skirt_side_l_parent, bone2Name=dyn_skirt_side_l_base, swing1LimitDegrees=15, swing2LimitDegrees=10, twistLimitDegrees=3}
]
