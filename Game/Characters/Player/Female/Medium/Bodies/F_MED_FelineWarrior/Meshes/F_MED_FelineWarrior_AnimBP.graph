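# Secondary-motion AnimBP for the FelineWarrior player body, layered on the pose copied from its attached parent mesh.
# Control Rig feeds sleeve and arm AnimDynamics, skirt RigidBody, head-twist-driven left-collar offsets, a post-dynamics Control Rig, then scarf RigidBody.
# State flags select skirt and scarf physics assets for front end, emote and crouch.
# The event graph updates those flags and drives material WPO cloth simulation, adding wind and turbulence while skydiving.
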AnimBlueprint /Game/Characters/Player/Female/Medium/Bodies/F_MED_FelineWarrior/Meshes/F_MED_FelineWarrior_AnimBP
target_skeleton=&/F_MED_FelineWarrior_Skeleton
graphs [
 AnimGraph {nodes [
   "Anim Dynamics - Chain: dyn_belt_loop_l -> dyn_belt_loop_l1"(Alpha=1, ExternalForce="0,0,0") {
    angularDampingOverride=1
    angularSpringConstant=100
    bChain=true
    bOverrideAngularDamping=true
    bUseGravityOverride=true
    bUsePlanarLimit=false
    bUseSphericalLimits=true
    simulationSpace=RootRelative
    alphaBoolBlend {blendOption=Linear}
    boundBone {boneName=dyn_belt_loop_l}
    chainEnd {boneName=dyn_belt_loop_l1}
    componentAppliedLinearAccClamp=<1000 1000 1000>
    gravityOverride {z=-900}
    physicsBodyDefinitions [
     {
      boundBone {boneName=dyn_belt_loop_l}
      boxExtents=<4.5 4.5 4.5>
      constraintSetup {angularLimitsMax=<20 20 20>, angularLimitsMin=<-20 -20 -20>, angularTarget=<1 0 0>}
      localJointOffset=<2.5481 0 0>
     }
     {
      boundBone {boneName=dyn_belt_loop_l2}
      boxExtents=<4.5 4.5 4.5>
      constraintSetup {angularLimitsMax=<20 20 20>, angularLimitsMin=<-20 -20 -20>, angularTarget=<1 0 0>}
      localJointOffset=<2.5481 0 0>
     }
     {
      boundBone {boneName=dyn_belt_loop_l1}
      boxExtents=<4.5 4.5 4.5>
      constraintSetup {angularLimitsMax=<20 20 20>, angularLimitsMin=<-20 -20 -20>, angularTarget=<1 0 0>}
      localJointOffset=<2.5481 0 0>
     }
    ]
    relativeSpaceBone {boneName=calf_twist_02_r}
    sphericalLimits [{limitRadius=31.5, drivingBone {boneName=dyn_skirt_side_frnt_l_01}, sphereLocalOffset=<-9.0678 12.8101 25.2001>}]
   }
   n1 = Control Rig!(Source=Copy Pose From Mesh() {bUseAttachedParent=true}, Alpha, cr:bool=Get bIsCrouching()) -> (Pose) {
    alphaBoolBlend {blendOption=Linear}
    controlRigAssetReference {blueprintRigClass=&/F_MED_FelineWarrior_CtrlRig.F_MED_FelineWarrior_CtrlRig_C}
   }
   n2 = Anim Dynamics - Bone: dyn_Sleeve_1_up_l(ComponentPose=Local To Component(LocalPose=@n1), Alpha=1, ExternalForce="0,0,0") {
    angularDampingOverride=1
    angularSpringConstant=5000
    bAngularSpring=true
    bGravityOverrideInSimSpace=true
    bOverrideAngularDamping=true
    gravityScale=0.35
    simulationSpace=RootRelative
    alphaBoolBlend {blendOption=Linear}
    alphaScaleBiasClamp {scale=0.2}
    boundBone {boneName=dyn_Sleeve_1_up_l}
    componentLinearAccScale {z=0.6}
    componentLinearVelScale {z=0.6}
    physicsBodyDefinitions [{
      collisionType=OuterSphere
      sphereCollisionRadius=4
      boundBone {boneName=dyn_Sleeve_1_up_l}
      boxExtents=<10 20 20>
      constraintSetup {angularLimitsMax=<45 0 10>, angularLimitsMin=<-45 -70 0>, angularTarget=<1 0 0>}
      localJointOffset=<7 0 -4.7292>
     }]
   }
   n3 = Anim Dynamics - Bone: dyn_Sleeve_1_down_l(ComponentPose=@n2, Alpha=1, ExternalForce="0,0,0") {
    angularDampingOverride=1
    angularSpringConstant=5000
    bAngularSpring=true
    bGravityOverrideInSimSpace=true
    bOverrideAngularDamping=true
    gravityScale=0.35
    simulationSpace=RootRelative
    alphaBoolBlend {blendOption=Linear}
    alphaScaleBiasClamp {scale=0.2}
    boundBone {boneName=dyn_Sleeve_1_down_l}
    componentLinearAccScale {z=0.6}
    componentLinearVelScale {z=0.6}
    physicsBodyDefinitions [{
      collisionType=OuterSphere
      sphereCollisionRadius=4
      boundBone {boneName=dyn_Sleeve_1_down_l}
      boxExtents=<10 20 20>
      constraintSetup {angularLimitsMax=<45 40 10>, angularLimitsMin=<-45 -40 -5>, angularTarget=<1 0 0>}
      localJointOffset=<7 0 6.8305>
     }]
   }
   n4 = Anim Dynamics - Bone: dyn_Sleeve_1_up_r(ComponentPose=@n3, Alpha=1, ExternalForce="0,0,0") {
    angularDampingOverride=1
    angularSpringConstant=5000
    bAngularSpring=true
    bGravityOverrideInSimSpace=true
    bOverrideAngularDamping=true
    gravityScale=0.35
    simulationSpace=RootRelative
    alphaBoolBlend {blendOption=Linear}
    alphaScaleBiasClamp {scale=0.2}
    boundBone {boneName=dyn_Sleeve_1_up_r}
    componentLinearAccScale {z=0.6}
    componentLinearVelScale {z=0.6}
    gravityOverride {z=-2000}
    physicsBodyDefinitions [{
      collisionType=OuterSphere
      sphereCollisionRadius=4
      boundBone {boneName=dyn_Sleeve_1_up_r}
      boxExtents=<10 20 20>
      constraintSetup {angularLimitsMax=<45 70 10>, angularLimitsMin=<-45 0 -10>, angularTarget=<1 0 0>}
      localJointOffset=<7 -0.4944 5.0649>
     }]
   }
   n5 = Anim Dynamics - Bone: dyn_Sleeve_1_down_r(ComponentPose=@n4, Alpha=1, ExternalForce="0,0,0") {
    angularDampingOverride=1
    angularSpringConstant=5000
    bAngularSpring=true
    bGravityOverrideInSimSpace=true
    bOverrideAngularDamping=true
    gravityScale=0.35
    simulationSpace=RootRelative
    alphaBoolBlend {blendOption=Linear}
    alphaScaleBiasClamp {scale=0.2}
    boundBone {boneName=dyn_Sleeve_1_down_r}
    componentLinearAccScale {z=0.6}
    componentLinearVelScale {z=0.6}
    gravityOverride {z=-2000}
    physicsBodyDefinitions [{
      collisionType=OuterSphere
      sphereCollisionRadius=4
      boundBone {boneName=dyn_Sleeve_1_down_r}
      boxExtents=<10 20 20>
      constraintSetup {angularLimitsMax=<45 40 10>, angularLimitsMin=<-45 -40 -10>, angularTarget=<1 0 0>}
      localJointOffset=<7 0 -6.6739>
     }]
   }
   n6 = Anim Dynamics - Bone: dyn_arm_l_loop(ComponentPose=@n5, Alpha=1, ExternalForce="0,0,0") {
    angularDampingOverride=1
    angularSpringConstant=100
    bOverrideAngularDamping=true
    bUseGravityOverride=true
    simulationSpace=RootRelative
    alphaBoolBlend {blendOption=Linear}
    boundBone {boneName=dyn_arm_l_loop}
    componentAppliedLinearAccClamp=<1000 1000 1000>
    componentLinearAccScale=<0.3 0.3 0.3>
    componentLinearVelScale=<0.3 0.3 0.3>
    gravityOverride {z=-900}
    physicsBodyDefinitions [{
      boundBone {boneName=dyn_arm_l_loop}
      boxExtents=<2 2 2>
      constraintSetup {angularLimitsMax=<2 15 15>, angularLimitsMin=<-2 0 -10>, angularTarget=<1 0 0>}
      localJointOffset=<0.2 0 0>
     }]
    relativeSpaceBone {boneName=calf_twist_02_r}
   }
   n7 = Anim Dynamics - Bone: dyn_arm_r_loop(ComponentPose=@n6, Alpha=1, ExternalForce="0,0,0") {
    angularDampingOverride=1
    angularSpringConstant=100
    bOverrideAngularDamping=true
    bUseGravityOverride=true
    simulationSpace=RootRelative
    alphaBoolBlend {blendOption=Linear}
    boundBone {boneName=dyn_arm_r_loop}
    componentAppliedLinearAccClamp=<1000 1000 1000>
    componentLinearAccScale=<0.3 0.3 0.3>
    componentLinearVelScale=<0.3 0.3 0.3>
    gravityOverride {z=-900}
    physicsBodyDefinitions [{
      boundBone {boneName=dyn_arm_r_loop}
      boxExtents=<2 2 2>
      constraintSetup {angularLimitsMax=<2 0 10>, angularLimitsMin=<-2 -15 -15>, angularTarget=<1 0 0>}
      localJointOffset=<0.2 0 0>
     }]
    relativeSpaceBone {boneName=calf_twist_02_r}
   }
   n8 = Anim Dynamics - Bone: dyn_arm_l_rope(ComponentPose=@n7, Alpha=1, ExternalForce="0,0,0") {
    angularDampingOverride=1
    angularSpringConstant=100
    bOverrideAngularDamping=true
    bUseGravityOverride=true
    simulationSpace=RootRelative
    alphaBoolBlend {blendOption=Linear}
    boundBone {boneName=dyn_arm_l_rope}
    componentAppliedLinearAccClamp=<1000 1000 1000>
    componentLinearAccScale=<0.3 0.3 0.3>
    componentLinearVelScale=<0.3 0.3 0.3>
    gravityOverride {z=-700}
    physicsBodyDefinitions [{
      boundBone {boneName=dyn_arm_l_rope}
      boxExtents=<2 2 2>
      constraintSetup {angularLimitsMax=<2 0 10>, angularLimitsMin=<-2 -10 0>, angularTarget=<1 0 0>}
      localJointOffset=<0.08 0 0>
     }]
    relativeSpaceBone {boneName=calf_twist_02_r}
   }
   n9 = Anim Dynamics - Bone: dyn_arm_r_rope(ComponentPose=@n8, Alpha=1, ExternalForce="0,0,0") {
    angularDampingOverride=1
    angularSpringConstant=100
    bOverrideAngularDamping=true
    bUseGravityOverride=true
    simulationSpace=RootRelative
    alphaBoolBlend {blendOption=Linear}
    boundBone {boneName=dyn_arm_r_rope}
    componentAppliedLinearAccClamp=<1000 1000 1000>
    componentLinearAccScale=<0.3 0.3 0.3>
    componentLinearVelScale=<0.3 0.3 0.3>
    gravityOverride {z=-400}
    physicsBodyDefinitions [{
      boundBone {boneName=dyn_arm_r_rope}
      boxExtents=<2 2 2>
      constraintSetup {angularLimitsMax=<2 0 5>, angularLimitsMin=<-2 -15 0>, angularTarget=<1 0 0>}
      localJointOffset=<0.1 0 0>
     }]
    relativeSpaceBone {boneName=calf_twist_02_r}
   }
   n10 = Get bIsInFrontEnd()
   n11 = RigidBody!(ComponentPose=@n9, bAlphaBoolEnabled:bool=NOT Boolean(A=@n10), ExternalForce) -> (Pose) {
    alphaInputType=Bool
    bOverrideWorldGravity=true
    overridePhysicsAsset=&/F_MED_FelineWarrior_skirt_Physics
    simulationSpace=BaseBoneSpace
    alphaBoolBlend {blendOption=Linear}
    baseBoneRef {boneName=pelvis}
    componentAppliedLinearAccClamp=<3000 3000 3000>
    componentLinearAccScale=<0.2 0.2 0.2>
    componentLinearVelScale=<0.2 0.2 0.2>
    overrideWorldGravity {z=-980}
    simSpaceSettings {maxAngularAcceleration=100, maxAngularVelocity=3, maxLinearAcceleration=100, maxLinearVelocity=100, velocityScaleZ=0, worldAlpha=1}
   }
   n12 = RigidBody(ComponentPose=@n11, bAlphaBoolEnabled=@n10) {
    alphaInputType=Bool
    bOverrideWorldGravity=true
    overridePhysicsAsset=&/F_MED_FelineWarrior_skirt_Physics_frontend
    simulationSpace=BaseBoneSpace
    alphaBoolBlend {blendOption=Linear}
    baseBoneRef {boneName=pelvis}
    componentAppliedLinearAccClamp=<3000 3000 3000>
    componentLinearAccScale=<0.2 0.2 0.2>
    componentLinearVelScale=<0.2 0.2 0.2>
    overrideWorldGravity {z=-980}
    simSpaceSettings {maxAngularAcceleration=100, maxAngularVelocity=3, maxLinearAcceleration=100, maxLinearVelocity=100, velocityScaleZ=0, worldAlpha=1}
   }
   n13 = Get Alpha Output() -> (Alpha Output:float)
   n14 = "Transform (Modify) Bone - Bone: front_collar_l"(ComponentPose=@n12, Alpha=@n13, Translation="0,0,0", Rotation="-26.687,0.025,0.018", Scale="1,1,1") {
    rotationMode=BMM_Additive
    rotationSpace=BCS_BoneSpace
    alphaBoolBlend {blendOption=Linear}
    boneToModify {boneName=front_collar_l}
    rotation=<rot -26.6867 0.0246 0.0182>
   }
   n15 = "Transform (Modify) Bone - Bone: side_collar_l"(ComponentPose=@n14, Alpha=@n13, Translation="0,0,0", Rotation="-8.123,-0.918,-14.223", Scale="1,1,1") {
    rotationMode=BMM_Additive
    rotationSpace=BCS_BoneSpace
    alphaBoolBlend {blendOption=Linear}
    boneToModify {boneName=side_collar_l}
    rotation=<rot -8.123 -0.918 -14.223>
   }
   n16 = Control Rig!(
    Source=Component To Local(ComponentPose=@n15)
    Alpha
    crouch:bool=Get onlycrouch() -> (onlycrouch:bool)
    kneslide:bool=Get bIsKneeSliding()
    basebalslide:bool=Get bIsBaseballSliding()
    cr=Get bIsCrouching()
   ) -> (Pose) {
    alphaBoolBlend {blendOption=Linear}
    controlRigAssetReference {blueprintRigClass=&/F_MED_FelineWarrior_postdyn_CtrlRig.F_MED_FelineWarrior_postdyn_CtrlRig_C}
   }
   n17 = RigidBody(ComponentPose=Local To Component(LocalPose=@n16), bAlphaBoolEnabled=Get not_fn_not_crouch() -> (not_fn_not_crouch:bool)) {
    alphaInputType=Bool
    bOverrideWorldGravity=true
    overridePhysicsAsset=&/F_MED_FelineWarrior__scarf_Physics
    simulationSpace=BaseBoneSpace
    alphaBoolBlend {blendInTime=0.7, blendOption=Linear, blendOutTime=0.7}
    baseBoneRef {boneName=spine_05}
    componentAppliedLinearAccClamp=<3000 3000 3000>
    componentLinearAccScale=<0.1 0.1 0.1>
    componentLinearVelScale=<0.1 0.1 0.1>
    overrideWorldGravity {z=-980}
   }
   n18 = RigidBody(ComponentPose=@n17, bAlphaBoolEnabled=Get bIsCrouching()) {
    alphaInputType=Bool
    bOverrideWorldGravity=true
    overridePhysicsAsset=&/F_MED_FelineWarrior__scarf_Physics_03
    simulationSpace=BaseBoneSpace
    alphaBoolBlend {blendInTime=0.1, blendOption=Linear, blendOutTime=0.4}
    baseBoneRef {boneName=spine_05}
    componentAppliedLinearAccClamp=<3000 3000 3000>
    componentLinearAccScale=<0.02 0.02 0.02>
    componentLinearVelScale=<0.02 0.02 0.02>
    overrideWorldGravity {z=-980}
   }
   n19 = RigidBody(ComponentPose=@n18, bAlphaBoolEnabled=Get fe_not_emote() -> (fe_not_emote:bool)) {
    alphaInputType=Bool
    bOverrideWorldGravity=true
    overridePhysicsAsset=&/F_MED_FelineWarrior__scarf_Physics_frontend_01
    simulationSpace=BaseBoneSpace
    alphaBoolBlend {blendInTime=0.7, blendOption=Linear, blendOutTime=0.7}
    baseBoneRef {boneName=spine_05}
    componentAppliedLinearAccClamp=<3000 3000 3000>
    componentLinearAccScale=<0.1 0.1 0.1>
    componentLinearVelScale=<0.1 0.1 0.1>
    overrideWorldGravity {z=-980}
   }
   n20 = RigidBody(ComponentPose=@n19, bAlphaBoolEnabled=Get feEmote() -> (feEmote:bool)) {
    alphaInputType=Bool
    bOverrideWorldGravity=true
    overridePhysicsAsset=&/F_MED_FelineWarrior__scarf_Physics_frontend
    simulationSpace=BaseBoneSpace
    alphaBoolBlend {blendInTime=0.7, blendOption=Linear, blendOutTime=0.7}
    baseBoneRef {boneName=spine_05}
    componentAppliedLinearAccClamp=<3000 3000 3000>
    componentLinearAccScale=<0.1 0.1 0.1>
    componentLinearVelScale=<0.1 0.1 0.1>
    overrideWorldGravity {z=-980}
   }
   Output Pose(Result=Component To Local(ComponentPose=@n20)) {layerGroup=None}
  ]}
 EventGraph {nodes [
   Set WPO Regional Function!(
    execute
    Material:Material Instance Dynamic Object Reference=Get dynamic_mat_instance() -> (dynamic_mat_instance:Material Instance Dynamic Object Reference)
    Overlay Material:Material Instance Dynamic Object Reference=Get dynamic_overlay_mat() -> (dynamic_overlay_mat:Material Instance Dynamic Object Reference)
    Skeletal Mesh:Skeletal Mesh Component Object Reference=Get mesh() -> (mesh:Skeletal Mesh Component Budgeted Object Reference)
    Velocity Transform:Name=calf_l
    Velocity Interp Speed:double=Get Velocity Interp Speed_reg2() -> (Velocity Interp Speed_reg2:double)
    Collision Transform:Name=calf_l
    Collision Transform Axis:EAxis Enum
    Invert Collision Transform Axis:bool=true
    Delta Time X:double=Get Delta Time X()
    Previous Velocity Transform Location:"Vector (by ref)"=Get Previous Velocity Transform Location_reg2() -> (Previous Velocity Transform Location_reg2:Vector)
    Previous Velocity Transform Velocity:"Vector (by ref)"=Get Previous Velocity Transform Velocity_reg2() -> (Previous Velocity Transform Velocity_reg2:Vector)
    Region:int=2
   ) -> (then)
   > Set WPO Regional Function(
    Material=Get dynamic_mat_instance()
    Overlay Material=Get dynamic_overlay_mat()
    Skeletal Mesh=Get mesh()
    Velocity Transform=pelvis
    Velocity Interp Speed=Get Velocity Interp Speed_reg3() -> (Velocity Interp Speed_reg3:double)
    Collision Transform=pelvis
    Invert Collision Transform Axis=false
    Delta Time X=Get Delta Time X()
    Previous Velocity Transform Location=Get Previous Velocity Transform Location_reg3() -> (Previous Velocity Transform Location_reg3:Vector)
    Previous Velocity Transform Velocity=Get Previous Velocity Transform Velocity_reg3() -> (Previous Velocity Transform Velocity_reg3:Vector)
    Region=3
   )
   n1 = Event BlueprintUpdateAnimation()
   > Set Delta Time X(Delta Time X=@n1.DeltaTimeX)
   > n2 = Sequence!(execute) -> (then_0, then_1, then_2:Exec)
   n3 = Get bIsMoving2D()
   n4 = "Absolute (Float)"(A=Get!() -> (PawnDeltaRotation_Roll:float, PawnDeltaRotation_Pitch:float, PawnDeltaRotation_Yaw:float).PawnDeltaRotation_Yaw)
   > Set onlycrouch(onlycrouch:bool=AND Boolean!(A=Get bIsCrouching(), B=NOT Boolean(A=@n3), C:bool="float <= float"(A=@n4, B=2.3) -> (ReturnValue=~)) -> (ReturnValue)) -> (Output_Get:bool)
   n5 = Get bIsInFrontEnd()
   n6 = Get bIsCrouching()
   > Set not_fn_not_crouch(not_fn_not_crouch:bool=AND Boolean!(A=NOT Boolean(A=@n5), B=NOT Boolean(A=@n6)) -> (ReturnValue)) -> (Output_Get:bool)
   > Set Alpha Output(Alpha Output:float=GetAlphaFromReferenceTransformDelta(
     SkeletalMeshComponent=GetOwnerSkeletalMeshComponent()
     BoneName=head
     BaseName=spine_05
     DeltaComponentType=QuaternionTwist
     InputRange=(X=0,Y=50)
     TwistAxis="1,0,0"
    )) -> (Output_Get:float)
   > n7 = Branch(Condition=Get bIsInFrontEnd())
   n8 = Get bIsPlayingEmote()
   > Set feEmote(feEmote:bool=@n8) -> (Output_Get:bool)
   > Set fe_not_emote(fe_not_emote:bool=NOT Boolean(A=@n8)) -> (Output_Get:bool)
   Set feEmote(execute=@n7.else, feEmote=false)
   > Set fe_not_emote(fe_not_emote=false)
   n9 = Branch(execute=@n2.then_1, Condition=Get bIsSkyDiving())
   > Set Velocity Interp Speed_reg1(Velocity Interp Speed_reg1:double=30) -> (Output_Get:double)
   > Set wind_strength(wind_strength:float=0.25) -> (Output_Get:float)
   > Set TurbulenceNoiseTileDivisor(TurbulenceNoiseTileDivisor:float=30) -> (Output_Get:float)
   Set Velocity Interp Speed_reg1(execute=@n9.else, Velocity Interp Speed_reg1=0)
   > Set wind_strength(wind_strength=0)
   > Set TurbulenceNoiseTileDivisor(TurbulenceNoiseTileDivisor=0)
   Set WPO Globals Function(
    execute=@n2.then_2
    Material=Get dynamic_mat_instance()
    Overlay Material=Get dynamic_overlay_mat()
    Skeletal Mesh=Get mesh()
    "Local Noise Transform  Location"="0, 0, 0"
    Local Noise Offset="0, 0, 0"
   )
   > Set WPO Regional Function(
    Material=Get dynamic_mat_instance()
    Overlay Material=Get dynamic_overlay_mat()
    Skeletal Mesh=Get mesh()
    Velocity Transform=calf_r
    Velocity Interp Speed=Get Velocity Interp Speed_reg1() -> (Velocity Interp Speed_reg1:double)
    Collision Transform=calf_r
    Invert Collision Transform Axis=false
    Delta Time X=Get Delta Time X()
    Previous Velocity Transform Location=Get Previous Velocity Transform Location_reg1() -> (Previous Velocity Transform Location_reg1:Vector)
    Previous Velocity Transform Velocity=Get Previous Velocity Transform Velocity_reg1() -> (Previous Velocity Transform Velocity_reg1:Vector)
    Region=1
   )
   n10 = Get dynamic_mat_instance()
   n11 = Get dynamic_overlay_mat()
   > SetScalarParameterValue(self=[@n10, @n11], ParameterName=ClothSim_WindStrength, Value=Get wind_strength() -> (wind_strength:float))
   > SetScalarParameterValue(
    self=[@n10, @n11]
    ParameterName=ClothSim_TurbulenceNoiseTileDivisor
    Value=Get TurbulenceNoiseTileDivisor() -> (TurbulenceNoiseTileDivisor:float)
   )
   n12 = Event OnCharacterCustomizationLoaded()
   > n13 = Set mesh(mesh:Skeletal Mesh Component Budgeted Object Reference=Get Array Item with EFort Custom Part Type(Array=@n12.CustomMeshParts)) -> (Output_Get:Skeletal Mesh Component Budgeted Object Reference)
   > n14 = Get Dynamic Material Instance(SkeletalMeshComponent=@n13)
   > Set dynamic_mat_instance(dynamic_mat_instance:Material Instance Dynamic Object Reference=@n14) -> (Output_Get:Material Instance Dynamic Object Reference)
   > n15 = GetOrCreateDynamicOverlayMaterial(MeshComponent=Get mesh(), MaterialName=None)
   > Set dynamic_overlay_mat(dynamic_overlay_mat:Material Instance Dynamic Object Reference=@n15) -> (Output_Get:Material Instance Dynamic Object Reference)
   n16 = Get bIsSliding()
   Set crouch_move_scarf(crouch_move_scarf:bool=AND Boolean!(A=Get bIsCrouching(), B=Get bIsMoving2D(), C:bool=NOT Boolean(A=@n16)) -> (ReturnValue)) -> (Output_Get:bool)
   n17 = Get bIsMoving2D()
   > Set crouch_notmove(crouch_notmove:bool=AND Boolean!(A=Get bIsCrouching(), B=NOT Boolean(A=@n17)) -> (ReturnValue)) -> (Output_Get:bool)
   n18 = "float >= float"(A="Absolute (Float)"(A=Get().PawnDeltaRotation_Yaw), B=2.3)
   > Set rotat(rotat:bool=AND Boolean(A=Get bIsCrouching(), B=@n18)) -> (Output_Get:bool)
   > Set onlycrouch(onlycrouch=AND Boolean(A=Get crouch_notmove() -> (crouch_notmove:bool), B=NOT Boolean(A=@n18)))
   n19 = Get bIsCrouching()
   n20 = Get bIsInFrontEnd()
   > Set not_fn_not_crouch(not_fn_not_crouch=AND Boolean(A=NOT Boolean(A=@n20), B=NOT Boolean(A=@n19)))
   Set Velocity Interp Speed_reg2(Velocity Interp Speed_reg2:double=50) -> (Output_Get:double)
   > Set Velocity Interp Speed_reg3(Velocity Interp Speed_reg3:double=5) -> (Output_Get:double)
   Set Velocity Interp Speed_reg2(Velocity Interp Speed_reg2=0)
   > Set Velocity Interp Speed_reg3(Velocity Interp Speed_reg3=0)
  ]}
]
variables [
 Alpha Output {type=number}
 crouch_move_scarf {type=boolean}
 crouch_notmove {type=boolean}
 Delta Time X {type=number}
 dynamic_mat_instance {type=MaterialInstanceDynamic}
 dynamic_overlay_mat {type=MaterialInstanceDynamic}
 fe_not_emote {type=boolean}
 feEmote {type=boolean}
 mesh {type=SkeletalMeshComponentBudgeted}
 not_fn_not_crouch {type=boolean}
 onlycrouch {type=boolean}
 Previous Velocity Transform Location_reg1 {type=Vector}
 Previous Velocity Transform Location_reg2 {type=Vector}
 Previous Velocity Transform Location_reg3 {type=Vector}
 Previous Velocity Transform Velocity_reg1 {type=Vector}
 Previous Velocity Transform Velocity_reg2 {type=Vector}
 Previous Velocity Transform Velocity_reg3 {type=Vector}
 rotat {type=boolean}
 TurbulenceNoiseTileDivisor {type=number}
 Velocity Interp Speed_reg1 {type=number, default=1}
 Velocity Interp Speed_reg2 {type=number, default=1}
 Velocity Interp Speed_reg3 {type=number, default=1}
 wind_strength {type=number}
]
